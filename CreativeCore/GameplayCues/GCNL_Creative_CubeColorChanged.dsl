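; Looping gameplay cue for the Creative color-changing cube that tints its Cascade particle effects to a team color.
; On application it runs the parent logic, then UpdateShaderParams casts the cue's instigator to the cube and requests a team color, using truncated NormalizedMagnitude as the team index.
; The color is stored as VectorizedColor and written to each particle component's "TeamColor" vector parameter.
; Removal only calls the parent.
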(event EventBeginPlay)

(event Collision|EventActorBeginOverlap
  (params (OtherActor :type Actor%20Object%20Reference)))

(event EventTick
  (params (DeltaSeconds :type Float%20%28single-precision%29)))

(event EventOnApplication(Cascade)
  (params (MyTarget :type Actor%20Object%20Reference) (Parameters :type Gameplay%20Cue%20Parameters%20Structure%20%28by%20ref%29) (ParticleComponents :type Array%20of%20Cascade%20Particle%20System%20Component%20Object%20References) (AudioComponents :type Array%20of%20Audio%20Component%20Object%20References) (BurstCameraShakeInstance :type Camera%20Shake%20Base%20Object%20Reference) (BurstDecalInstance :type Decal%20Actor%20Object%20Reference))
  (|CallParentFunction MyTarget Parameters ParticleComponents AudioComponents BurstCameraShakeInstance BurstDecalInstance)
  (CallFunction|UpdateShaderParams Parameters ParticleComponents))

(event EventOnRemoval(Cascade)
  (params (MyTarget :type Actor%20Object%20Reference) (Parameters :type Gameplay%20Cue%20Parameters%20Structure%20%28by%20ref%29) (ParticleComponents :type Array%20of%20Cascade%20Particle%20System%20Component%20Object%20References) (AudioComponents :type Array%20of%20Audio%20Component%20Object%20References) (BurstCameraShakeInstance :type Camera%20Shake%20Base%20Object%20Reference) (BurstDecalInstance :type Decal%20Actor%20Object%20Reference))
  (|CallParentFunction MyTarget Parameters ParticleComponents AudioComponents BurstCameraShakeInstance BurstDecalInstance))

(fn UpdateShaderParams
  (params (Parameters :type Gameplay%20Cue%20Parameters%20Structure%20%28by%20ref%29) (Array :type Array%20of%20Cascade%20Particle%20System%20Component%20Object%20References))
  (bind (_normalizedmagnitude _rawmagnitude _effectcontext _matchedtagname _originaltag _aggregatedsourcetags _aggregatedtargettags _location _normal _instigator) (Ability|GameplayCue|BreakGameplayCueParameters Parameters))
  (bind _aspg_cube_color_changing (Utilities|Casting|CastToPG_Cube_ColorChanging _instigator)
    (:then
      (Variables|Default|SetVectorizedColor (Math|Conversions|ToVector(LinearColor) (Class|PGCubeColorChanging|GetColorforTeam _aspg_cube_color_changing (Math|Float|Truncate _normalizedmagnitude)))))
    (:CastFailed
      _))
  (for _array_element Array
    (Effects|Components|ParticleSystem|SetVectorParameter _array_element "TeamColor" (Variables|Default|GetVectorizedColor))))

(fn ConstructionScript)
