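; Looping gameplay cue that gives a player pawn grinding feet effects and audio, using Creative GrindTrap lightning VFX.
; On application it attaches team-colored lightning to both feet, and on looping start it plays a first- or third-person grind loop and binds actor hits to collision audio.
; Tick drives the loop's Speed parameter from velocity, zeroing it when not walking.
; Removal deactivates particles, unbinds hits, destroys the collision component and fades the loop.
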(event EventTick
  (params (DeltaSeconds :type Float%20%28single-precision%29))
  (bind _current_grind_loop_audio (Variables|Audio|GetCurrent%20Grind%20Loop%20Audio))
  (bind _target_pawn (Variables|Default|GetTarget%20Pawn)
    (:then
      (Utilities|IsValid _current_grind_loop_audio
        (:Is%20Valid
          (if (or (AI|Components|NavMovement|IsFalling (Variables|/Script/Engine.Character|GetCharacterMovement (stale EventOnLoopingStart(Cascade) _asfort_player_pawn_athena))) (not (Utilities|Enum|Equal(Enum) (Variables|/Script/Engine.CharacterMovementComponent|GetMovementMode (Variables|/Script/Engine.Character|GetCharacterMovement (stale EventOnLoopingStart(Cascade) _asfort_player_pawn_athena))) "MOVE_Walking")))
            ;[1 1 1] Lowest current is .45
            (Audio|Parameter|SetFloatParameter _current_grind_loop_audio "Speed" 0.0 :!class /Script/AudioExtensions.AudioParameterControllerInterface)
            (else
              (Audio|Parameter|SetFloatParameter _current_grind_loop_audio "Speed" (Math|Float|MapRangeClamped (Math|Vector|VectorLength (Transformation|GetVelocity _target_pawn)) 0.0 8e+02 0.3 1.5) :!class /Script/AudioExtensions.AudioParameterControllerInterface))))))
    (:else)))

(event EventOnLoopingStart(Cascade)
  (params (MyTarget :type Actor%20Object%20Reference) (Parameters :type Gameplay%20Cue%20Parameters%20Structure%20%28by%20ref%29) (ParticleComponents :type Array%20of%20Cascade%20Particle%20System%20Component%20Object%20References) (AudioComponents :type Array%20of%20Audio%20Component%20Object%20References))
  (bind _output_get (Variables|Default|SetTarget%20Pawn MyTarget))
  (bind _asfort_player_pawn_athena (Utilities|Casting|CastToFortPlayerPawnAthena _output_get))
  (bind _returnvalue (AddComponent|NODE_AddFortCollisionAudioComponent-0))
  (Variables|Default|SetCollisionAudioComp _returnvalue)
  (Collision|BindEventtoOnActorHit _asfort_player_pawn_athena (EventDispatchers|CreateEvent :!event Custom|On%20Target%20Pawn%20Hit%20Event))
  (Audio|SetCollisionSoundatIndex (Variables|Default|GetCollisionAudioComp) 0 (Variables|Audio|GetCollision%20Audio))
  (bind _returnvalue_2 (Audio|SpawnSoundAttached (select (Pawn|IsLocallyControlled _asfort_player_pawn_athena) (Variables|Audio|GetGrind%201p%20Loop) (Variables|Audio|GetGrind%203p%20Loop)) (Variables|/Script/Engine.Actor|GetRootComponent _asfort_player_pawn_athena)))
  (Variables|Audio|SetCurrent%20Grind%20Loop%20Audio _returnvalue_2))

(event EventOnRemoval(Cascade)
  (params (MyTarget :type Actor%20Object%20Reference) (Parameters :type Gameplay%20Cue%20Parameters%20Structure%20%28by%20ref%29) (ParticleComponents :type Array%20of%20Cascade%20Particle%20System%20Component%20Object%20References) (AudioComponents :type Array%20of%20Audio%20Component%20Object%20References) (BurstCameraShakeInstance :type Camera%20Shake%20Base%20Object%20Reference) (BurstDecalInstance :type Decal%20Actor%20Object%20Reference))
  (bind _current_grind_loop_audio (Variables|Audio|GetCurrent%20Grind%20Loop%20Audio))
  (CallFunction|DeactivateFootParticles)
  (bind _asfort_player_pawn_athena (Utilities|Casting|CastToFortPlayerPawnAthena MyTarget))
  (Collision|UnbindEventfromOnActorHit _asfort_player_pawn_athena (EventDispatchers|CreateEvent :!event Custom|On%20Target%20Pawn%20Hit%20Event))
  (Components|DestroyComponent (Variables|Default|GetCollisionAudioComp) self)
  (Utilities|IsValid _current_grind_loop_audio
    (:Is%20Valid
      (Audio|Components|Audio|FadeOut _current_grind_loop_audio 0.75 0.0 "Linear" :!class /Script/Engine.AudioComponent))))

(event EventOnApplication(Cascade)
  (params (MyTarget :type Actor%20Object%20Reference) (Parameters :type Gameplay%20Cue%20Parameters%20Structure%20%28by%20ref%29) (ParticleComponents :type Array%20of%20Cascade%20Particle%20System%20Component%20Object%20References) (AudioComponents :type Array%20of%20Audio%20Component%20Object%20References) (BurstCameraShakeInstance :type Camera%20Shake%20Base%20Object%20Reference) (BurstDecalInstance :type Decal%20Actor%20Object%20Reference))
  (CallFunction|DeactivateFootParticles)
  (bind _asfort_player_pawn (Utilities|Casting|CastToFortPlayerPawn MyTarget))
  (bind _returnvalue (Effects|SpawnEmitterAttached :EmitterTemplate "/Game/Creative/Devices/GrindTrap/VFX/P_GrindTrap_FeetLightning.P_GrindTrap_FeetLightning" :AttachToComponent (Variables|/Script/Engine.Character|GetMesh _asfort_player_pawn) :AttachPointName "FX_foot_lSocket" :Scale "1.000000,1.000000,1.000000" :LocationType "KeepRelativeOffset" :bAutoDestroy true :PoolingMethod "AutoRelease" :bAutoActivate true))
  (bind _output_get (Variables|Default|SetLeftFootFX _returnvalue))
  (Effects|Components|ParticleSystem|SetVectorParameter _output_get "TeamColor" (CallFunction|GetPawnTeamNumber _asfort_player_pawn))
  (bind _returnvalue_2 (Effects|SpawnEmitterAttached :EmitterTemplate "/Game/Creative/Devices/GrindTrap/VFX/P_GrindTrap_FeetLightning.P_GrindTrap_FeetLightning" :AttachToComponent (Variables|/Script/Engine.Character|GetMesh _asfort_player_pawn) :AttachPointName "FX_foot_rSocket" :Scale "1.000000,1.000000,1.000000" :LocationType "KeepRelativeOffset" :bAutoDestroy true :PoolingMethod "AutoRelease" :bAutoActivate true))
  (bind _output_get_2 (Variables|Default|SetRightFootFX _returnvalue_2))
  (Effects|Components|ParticleSystem|SetVectorParameter _output_get_2 "TeamColor" (CallFunction|GetPawnTeamNumber _asfort_player_pawn)))

(event Custom|On%20Target%20Pawn%20Hit%20Event
  (params (SelfActor :type Actor%20Object%20Reference) (OtherActor :type Actor%20Object%20Reference) (NormalImpulse :type Vector) (Hit :type Hit%20Result%20Structure%20%28by%20ref%29))
  (bind (_bblockinghit _binitialoverlap _time _distance _location _impactpoint _normal _impactnormal) (Collision|BreakHitResult Hit))
  ;[1 1 1] On Audio Collision
  (Audio|OnCollision (Variables|Default|GetCollisionAudioComp) _location (* _impactnormal (Transformation|GetVelocity (Variables|Default|GetTarget%20Pawn)))))

(fn DeactivateFootParticles
  (Utilities|FlowControl|Sequence
    (:then_0
      (bind _leftfootfx (Variables|Default|GetLeftFootFX)
        (:then
          (Components|Activation|Deactivate _leftfootfx))
        (:else)))
    (:then_1
      (bind _rightfootfx (Variables|Default|GetRightFootFX)
        (:then
          (Components|Activation|Deactivate _rightfootfx))
        (:else)))))

(fn GetPawnTeamNumber
  (params (Actor :type Actor%20Object%20Reference))
  (return (Math|Conversions|ToVector(LinearColor) (Utilities|Array|Get(acopy) (Class|GetClassDefaults "/Game/Creative/Sets/MilitaryBase/Props/CreativeTeamColors.CreativeTeamColors_C") (- (Math|Conversions|ToInteger(Byte) (Team|GetActorTeam Actor)) 2)))))

(fn ConstructionScript)
