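; Looping gameplay cue that plays audio, particle and camera feedback while a player or vehicle rides a Creative vortex (wind tunnel).
; For local players, a timer polls vortex direction, falling back to the vortex register, avoiding costly per-tick lookups.
; Tick uses target velocity to drive attached particles, shake scale and loop audio.
; A timeline fades tunnel opacity through a material parameter collection; removal reverses it, playing exit effects.
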(event Custom|Tick%20Test%20Current%20Vortex
  ;[1 0.715 0] Every 'VortexTextUpdateRate' check to see if the vector has updated (potentially expensive, so don't run on tick)
  (bind _vortexactor (Variables|Default|GetVortexActor)
    (:then
      (bind (_relevant _direction) (Vortex|GetVortexDirection(Message) _vortexactor (Variables|Default|GetMyTarget)))
      (if _relevant
        (Variables|Default|SetVortexDirection _direction)
        (else
          _)))
    (:else
      _))
  (bind _direction_2 (CallFunction|FindActiveVortex))
  (Variables|Default|SetVortexDirection _direction_2))

(event EventTick
  (params (DeltaSeconds :type Float%20%28single-precision%29))
  (bind _vortexdirection (Variables|Default|GetVortexDirection))
  (bind _velocitymult (Variables|Default|GetVelocityMult))
  (bind _mytarget (Variables|Default|GetMyTarget))
  (bind _returnvalue (Transformation|GetVelocity _mytarget))
  (bind _returnvalue_2 (* _returnvalue _velocitymult))
  (bind _returnvalue_3 (Math|Vector|ClampVectorSize _returnvalue_2 -5e+02 5e+02))
  (bind _returnvalue_4 (Math|Vector|DotProduct _returnvalue_3 _vortexdirection))
  (bind _vortexdirection_2 (Variables|Default|GetVortexDirection))
  (bind _returnvalue_5 (* _vortexdirection_2 _returnvalue_4))
  (bind _mytarget_2 (Variables|Default|GetMyTarget)
    (:then
      (CallFunction|UpdateIsLocalandIsPlayer _mytarget_2)
      (if (Variables|Default|GetbShouldEnableAudio%20)
        (CallFunction|UpdateAudio))
      (if (Utilities|Operators|NotEqual(!=) :A (Variables|Default|GetVortexDirection))
        ;[1 1 1] Update attached player FX
        (bind _wind_tunnel_particle (Variables|Default|GetWind%20Tunnel%20Particle)
          (:then
            (Effects|Components|ParticleSystem|SetVectorParameter _wind_tunnel_particle "Velocity" _returnvalue_5)
            (Effects|Components|ParticleSystem|SetVectorParameter _wind_tunnel_particle "Velocity2" (* _returnvalue_5 -0.5))
            (Effects|Components|ParticleSystem|SetFloatParameter _wind_tunnel_particle "Opacity" (Math|Float|MapRangeClamped (Math|Float|Absolute(Float) _returnvalue_4) 3e+02 5e+02 0.1 1.0))
            ;[1 1 1] Scale camera shake magnitude
            (bind _camerashake (Variables|Default|GetCameraShake)
              (:then
                (Variables|/Script/EngineCameras.LegacyCameraShake|SetShakeScale (Math|Float|MapRangeClamped (Math|Vector|VectorLength (Transformation|GetVelocity (Variables|Default|GetMyTarget))) (ScalableFloat|GetValueatLevel (Variables|Default|GetDataPassiveModeLateralSpeed)) (ScalableFloat|GetValueatLevel (Variables|Default|GetDataActiveModeLateralSpeed)) (Variables|Default|GetShakeMin) (Variables|Default|GetShakeMax)) _camerashake))
              (:else)))
          (:else))))
    (:else)))

(event EventOnLoopingStart(Cascade)
  (params (MyTarget :type Actor%20Object%20Reference) (Parameters :type Gameplay%20Cue%20Parameters%20Structure%20%28by%20ref%29) (ParticleComponents :type Array%20of%20Cascade%20Particle%20System%20Component%20Object%20References) (AudioComponents :type Array%20of%20Audio%20Component%20Object%20References))
  (bind (_normalizedmagnitude _rawmagnitude _effectcontext) (Ability|GameplayCue|BreakGameplayCueParameters Parameters))
  (CallFunction|Init MyTarget)
  (Variables|Default|SetVortexActor (Ability|EffectContext|GetInstigatorActor _effectcontext))
  (if (Variables|Default|GetbShouldEnableAudio%20)
    (CallFunction|StartOrStopAudio true))
  (Utilities|FlowControl|Sequence
    (:then_0
      (if (and (Variables|Default|GetIsTargetLocal) (Variables|Default|GetIsTargetPlayer))
        (CallFunction|TickTestCurrentVortex)
        (bind _returnvalue (Utilities|Time|SetTimerbyEvent (EventDispatchers|CreateEvent :!event Custom|Tick%20Test%20Current%20Vortex) (Variables|Default|GetVortex%20Test%20Update%20Rate) true))
        (Variables|Default|SetTick%20Update%20Handle _returnvalue)
        (CallFunction|SpawnLocalTunnelEffect)))
    (:then_1
      ;[1 1 1] Make tunnels more opaue for local player riding in them.  Play camera shake.
      (if (Variables|Default|GetIsTargetLocal)
        (CallFunction|StartCameraShake)
        (bind (_direction _newtrack_0) (Timeline|OpacityFade Play)
          (:Update
            (Rendering|Material|SetScalarParameterValue 0 "/Game/Athena/Environments/Blueprints/SkyTube/MPC_Gameplay_BPs.MPC_Gameplay_BPs" "WindTunnelOpacityAdd" _newtrack_0 :!class /Script/Engine.KismetMaterialLibrary)))))))

(event EventOnRemoval(Cascade)
  (params (MyTarget :type Actor%20Object%20Reference) (Parameters :type Gameplay%20Cue%20Parameters%20Structure%20%28by%20ref%29) (ParticleComponents :type Array%20of%20Cascade%20Particle%20System%20Component%20Object%20References) (AudioComponents :type Array%20of%20Audio%20Component%20Object%20References) (BurstCameraShakeInstance :type Camera%20Shake%20Base%20Object%20Reference) (BurstDecalInstance :type Decal%20Actor%20Object%20Reference))
  (Utilities|Time|ClearandInvalidateTimerbyHandle 0 (Variables|Default|GetTick%20Update%20Handle))
  (CallFunction|UpdateIsLocalandIsPlayer MyTarget)
  (bind _wind_tunnel_particle (Variables|Default|GetWind%20Tunnel%20Particle)
    (:then
      (Components|DestroyComponent _wind_tunnel_particle 0))
    (:else
      _))
  (if (Variables|Default|GetbShouldEnableAudio%20)
    (CallFunction|StartOrStopAudio false))
  (bind _asfort_player_pawn (Utilities|Casting|CastToFortPlayerPawn MyTarget)
    (:then
      (CallFunction|StopCameraShake _asfort_player_pawn)
      (Effects|SpawnEmitterAttached "/FNGameplayCues/Athena/Environments/Blueprints/SkyTube/FX/P_Wind_Tunnel_Attached_Exit_Player.P_Wind_Tunnel_Attached_Exit_Player" (Variables|/Script/Engine.Character|GetMesh _asfort_player_pawn) "pelvis"))
    (:CastFailed
      _))
  ;[1 1 1] Make tunnels more opaue for local player riding in them.  Play camera shake.
  (if (Variables|Default|GetIsTargetLocal)
    (Timeline|OpacityFade Reverse)))

(event Custom|StopShake
  (CallFunction|StopCameraShake (Variables|Default|GetMyTargetPlayer)))

(event Custom|Update%20Audio
  (bind _loopingaudio (Variables|Default|GetLoopingAudio))
  (if (Variables|Default|GetIsAudioEnabled)
    (Utilities|FlowControl|Sequence
      (:then_0
        (Utilities|IsValid _loopingaudio
          (:Is%20Valid
            (Audio|Components|Audio|SetVolumeMultiplier _loopingaudio (Math|Interpolation|FInterpTo (Variables|/Script/Engine.AudioComponent|GetVolumeMultiplier _loopingaudio) (select (CallFunction|GetIsPlayerinVehicle) 0.1 1.0) (Utilities|Time|GetWorldDeltaSeconds 0) 5.0) :!class /Script/Engine.AudioComponent))))
      (:then_1
        (Audio|Parameter|SetIntegerParameter (Variables|Default|GetLoopingAudio) "IsLocal" (Math|Conversions|ToInteger(Boolean) (Variables|Default|GetIsTargetLocal)) :!class /Script/AudioExtensions.AudioParameterControllerInterface)
        (Audio|Parameter|SetIntegerParameter (Variables|Default|GetLoopingAudio) "IsVehicleEmpty" (Math|Conversions|ToInteger(Boolean) (Variables|Default|GetIsVehicleEmpty)) :!class /Script/AudioExtensions.AudioParameterControllerInterface)
        ; Update Speed Param
        (bind _output_get (Variables|Sound|SetAudioCurrentSpeedParam (Math|Interpolation|FInterpTo (Variables|Sound|GetAudioCurrentSpeedParam) (Math|Float|MapRangeClamped (Math|Vector|VectorLength (Transformation|GetVelocity (Variables|Default|GetMyTarget))) 1.5e+03 2.5e+03 0.0 1.0) (Utilities|Time|GetWorldDeltaSeconds 0) 5.0)))
        (Audio|Parameter|SetFloatParameter (Variables|Default|GetLoopingAudio) "Speed" _output_get :!class /Script/AudioExtensions.AudioParameterControllerInterface)))))

(event Custom|Start%20or%20Stop%20Audio
  (params (Start :type Boolean))
  (bind _mytarget (Variables|Default|GetMyTarget))
  (bind _rootcomponent (Variables|/Script/Engine.Actor|GetRootComponent _mytarget))
  (bind _output_get (Variables|Default|SetIsAudioEnabled Start))
  (if _output_get
    (Variables|Default|GetLoopingAudio
      (:then)
      (:else
        (bind _returnvalue (Audio|SpawnSoundAttached (select (Variables|Default|GetIsTargetPlayer) (Variables|Sound|GetSoundPlayerLoop) (Variables|Sound|GetSoundVehicleLoop)) _rootcomponent))
        (bind _output_get_2 (Variables|Default|SetLoopingAudio _returnvalue))
        (Audio|Components|Audio|FadeIn _output_get_2 1.0 1.0 0.0 "Linear" :!class /Script/Engine.AudioComponent)
        (Audio|Parameter|SetIntegerParameter (Variables|Default|GetLoopingAudio) "IsLocal" (Math|Conversions|ToInteger(Boolean) (Variables|Default|GetIsTargetLocal)) :!class /Script/AudioExtensions.AudioParameterControllerInterface)
        (bind _returnvalue_2 (Audio|SpawnSoundAttached (select (Variables|Default|GetIsTargetPlayer) (Variables|Sound|GetSoundPlayerEnter) (Variables|Sound|GetSoundVehicleEnter)) _rootcomponent))
        (Audio|Parameter|SetIntegerParameter _returnvalue_2 "IsLocal" (Math|Conversions|ToInteger(Boolean) (Variables|Default|GetIsTargetLocal)) :!class /Script/AudioExtensions.AudioParameterControllerInterface)
        (Audio|Parameter|SetIntegerParameter _returnvalue_2 "IsVehicleEmpty" (Math|Conversions|ToInteger(Boolean) (Variables|Default|GetIsVehicleEmpty)) :!class /Script/AudioExtensions.AudioParameterControllerInterface)))
    (else
      (bind _loopingaudio (Variables|Default|GetLoopingAudio)
        (:then
          (Audio|Components|Audio|FadeOut _loopingaudio 1.0 0.0 "Linear" :!class /Script/Engine.AudioComponent)
          (Audio|Parameter|SetIntegerParameter (Variables|Default|GetLoopingAudio) "IsLocal" (Math|Conversions|ToInteger(Boolean) (Variables|Default|GetIsTargetLocal)) :!class /Script/AudioExtensions.AudioParameterControllerInterface)
          (bind _returnvalue_3 (Audio|SpawnSoundAttached (select (Variables|Default|GetIsTargetPlayer) (Variables|Sound|GetSoundPlayerExit) (Variables|Sound|GetSoundVehicleExit)) _rootcomponent))
          (Audio|Parameter|SetIntegerParameter _returnvalue_3 "IsLocal" (Math|Conversions|ToInteger(Boolean) (Variables|Default|GetIsTargetLocal)) :!class /Script/AudioExtensions.AudioParameterControllerInterface)
          (Audio|Parameter|SetIntegerParameter _returnvalue_3 "IsVehicleEmpty" (Math|Conversions|ToInteger(Boolean) (Variables|Default|GetIsVehicleEmpty)) :!class /Script/AudioExtensions.AudioParameterControllerInterface))
        (:else)))))

(fn FindActiveVortex
  ;[0 0.8734 1 0.25] Updates the Active Vortex from a list of all the vortices in the volume.
  ; Early outs as soon as it finds a valid vortex.
  (Variables|Default|GetMyTargetPlayer
    (:then
      (bind (_bisvalid _register) (Class|FortCreativeIndexedObjectRegisterComponent|GetRegisteratIndex (Variables|Default|GetVortex%20Register) 0))
      (bind _register_objects (Utilities|Struct|BreakFortObjectArray _register))
      (if _bisvalid
        (for _array_element _register_objects
          (bind (_relevant _direction) (Vortex|GetVortexDirection(Message) _array_element (Variables|Default|GetMyTarget)))
          (if _relevant
            (bind _asactor (Utilities|Casting|CastToActor _array_element))
            (Variables|Default|SetVortexActor _asactor)
            (return _direction)))
        (return "0, 0, 0")))
    (:else)))

(fn GetIsPlayerInVehicle
  (bind _asfort_pawn (Utilities|Casting|CastToFortPawn (Variables|Default|GetMyTarget))
    (:then
      (Utilities|IsinVehicle _asfort_pawn
        (:true
          (return true))
        (:false
          _)))
    (:CastFailed
      _))
  (return false))

(fn Init
  (params (MyTarget :type Actor%20Object%20Reference))
  (bind _output_get (Variables|Default|SetMyTarget MyTarget))
  (CallFunction|UpdateIsLocalandIsPlayer _output_get)
  (bind _mytarget (Variables|Default|GetMyTarget)
    (:then
      (CreativeObjectRegistry|GetOrCreateObjectRegister _mytarget "(TagName=\"Creative.Register.Vortex\")" false 0 (Variables|Default|GetVortex%20Register) _mytarget))
    (:else)))

(fn SpawnLocalTunnelEffect
  (bind _mytarget (Variables|Default|GetMyTarget)
    (:then
      (bind _asfort_player_pawn (Utilities|Casting|CastToFortPlayerPawn _mytarget))
      (bind _returnvalue (Effects|SpawnEmitterAttached :EmitterTemplate "/FNGameplayCues/Athena/Environments/Blueprints/SkyTube/FX/P_Wind_Tunnel_Attached_Loop_Player.P_Wind_Tunnel_Attached_Loop_Player" :AttachToComponent (Variables|/Script/Engine.Character|GetMesh _asfort_player_pawn) :AttachPointName "None" :Scale "1.000000,1.000000,1.000000" :LocationType "SnapToTarget" :bAutoDestroy true :PoolingMethod "None" :bAutoActivate true))
      (Variables|Default|SetWind%20Tunnel%20Particle _returnvalue))
    (:else)))

(fn StartCameraShake
  (if (Variables|Default|GetIsTargetPlayer)
    (bind _output_get (Variables|Default|SetAppliedShakeToController (Fortnite|GetFortPlayerControllerfromActor (Variables|Default|GetMyTarget))))
    (bind _returnvalue (CameraShakes|StartCameraShake :self (Variables|/Script/Engine.PlayerController|GetPlayerCameraManager _output_get) :ShakeClass "/FNGameplayCues/Athena/Environments/Blueprints/SkyTube/CameraShake_SkytubeVelocity.CameraShake_SkytubeVelocity_C" :Scale 1.0 :PlaySpace "CameraLocal"))
    (Variables|Default|SetCameraShake (CameraShakes|ConvLegacyCameraShake _returnvalue))
    ;[1 1 1] Remove the shake when the player exits this pawn (as it is a vehicle)
    (if (Utilities|DoesObjectImplementInterface (Variables|Default|GetMyTarget) "/Script/FortniteGame.FortVehicleInterface")
      (Vehicle|BindEventtoOnExitVehicle (Variables|Default|GetAppliedShakeToController) (EventDispatchers|CreateEvent :!event StopShake)))))

(fn StopCameraShake
  (params (Target :type Fort%20Player%20Pawn%20Object%20Reference))
  ;[1 1 1] Remove this callback when we stop the shake!
  (bind _appliedshaketocontroller (Variables|Default|GetAppliedShakeToController)
    (:then
      (Vehicle|UnbindEventfromOnExitVehicle _appliedshaketocontroller (EventDispatchers|CreateEvent :!event StopShake))
      (bind _camerashake (Variables|Default|GetCameraShake)
        (:then
          (CameraShakes|StopCameraShake (Variables|/Script/Engine.PlayerController|GetPlayerCameraManager (Variables|Default|GetAppliedShakeToController)) _camerashake false))
        (:else)))
    (:else)))

(fn UpdateIsLocalAndIsPlayer
  (params (Target :type Actor%20Object%20Reference))
  (bind _asfort_player_pawn (Utilities|Casting|CastToFortPlayerPawn Target)
    (:then
      (Variables|Default|SetMyTargetPlayer _asfort_player_pawn)
      (Variables|Default|SetIsTargetPlayer true)
      (Variables|Default|SetIsTargetLocal (Pawn|IsLocallyControlled _asfort_player_pawn)))
    (:CastFailed
      (Variables|Default|SetIsTargetPlayer false)
      (bind _asfort_vehicle_interface (Utilities|Casting|CastToFortVehicleInterface Target))
      (if (> (Utilities|Array|Length (Vehicle|GetAllPassengers _asfort_vehicle_interface)) 0)
        (Variables|Default|SetIsVehicleEmpty false)
        (bind _array_element (Utilities|Array|ForEachLoopwithBreak (Vehicle|GetAllPassengers _asfort_vehicle_interface))
          (:LoopBody
            (bind _output_get (Variables|Default|SetIsTargetLocal (Pawn|IsLocallyControlled _array_element)))
            (if _output_get
              _)))
        (else
          (Variables|Default|SetIsTargetLocal false)
          (Variables|Default|SetIsVehicleEmpty true))))))

(fn ConstructionScript)
