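; Looping gameplay cue for Creative invisibility that maintains material state on the affected player pawn.
; On application it caches the pawn and binds to part changes, setting MovementOpacity to 1 on each changed part's meshes.
; On removal it unbinds and, if the target is the cached pawn, resets SpookyMistMode in the global Fortnite material parameter collection.
; A hit handler feeds impact location and velocity-scaled normal to collision audio.
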(event Custom|BindOnPartChanged
  ;[1 1 1] Character part changed
  (EventDispatchers|AssignOnPartChanged (Variables|Default|GetPlayerPawn) (delegate Custom|PartChanged)))

(event Custom|PartChanged
  (params (PartType :type EFortCustomPartType%20Enum :default "Head"))
  ;[1 1 1] Character part changed
  (for _array_element (Customization|GetSkinnedMeshComponentsforPartType (Variables|Default|GetPlayerPawn) PartType)
    (Rendering|Material|SetScalarParameterValueonMaterials _array_element "MovementOpacity" 1.0)))

(event Custom|UnbindPartChanged
  ;[1 1 1] Character part changed
  (EventDispatchers|UnbindEventfromOnPartChanged (Variables|Default|GetPlayerPawn) (delegate Custom|PartChanged)))

(event EventOnApplication(Cascade)
  (params (MyTarget :type Actor%20Object%20Reference) (Parameters :type Gameplay%20Cue%20Parameters%20Structure%20%28by%20ref%29) (ParticleComponents :type Array%20of%20Cascade%20Particle%20System%20Component%20Object%20References) (AudioComponents :type Array%20of%20Audio%20Component%20Object%20References) (BurstCameraShakeInstance :type Camera%20Shake%20Base%20Object%20Reference) (BurstDecalInstance :type Decal%20Actor%20Object%20Reference))
  ;[1 1 1] Application
  (bind _asfort_player_pawn (Utilities|Casting|CastToFortPlayerPawn MyTarget))
  (Variables|Default|SetPlayerPawn _asfort_player_pawn)
  (Utilities|CollectSkeletalMesh
    (:Valid
      (CallFunction|BindonPartChanged))))

(event EventOnRemoval(Cascade)
  (params (MyTarget :type Actor%20Object%20Reference) (Parameters :type Gameplay%20Cue%20Parameters%20Structure%20%28by%20ref%29) (ParticleComponents :type Array%20of%20Cascade%20Particle%20System%20Component%20Object%20References) (AudioComponents :type Array%20of%20Audio%20Component%20Object%20References) (BurstCameraShakeInstance :type Camera%20Shake%20Base%20Object%20Reference) (BurstDecalInstance :type Decal%20Actor%20Object%20Reference))
  (bind _playerpawn (Variables|Default|GetPlayerPawn))
  ;[1 1 1] Removal
  (CallFunction|UnbindPartChanged)
  (if (Utilities|Equal(Object) MyTarget _playerpawn)
    (Rendering|Material|SetScalarParameterValue _playerpawn "/Game/Packages/Fortress_SharedMaterials/GlobalMaterialParameters/FortniteMaterialParameters.FortniteMaterialParameters" "SpookyMistMode" :!class /Script/Engine.KismetMaterialLibrary)))

(event Custom|On%20Target%20Pawn%20Hit%20Event
  (params (SelfActor :type Actor%20Object%20Reference) (OtherActor :type Actor%20Object%20Reference) (NormalImpulse :type Vector) (Hit :type Hit%20Result%20Structure%20%28by%20ref%29))
  (bind (_bblockinghit _binitialoverlap _time _distance _location _impactpoint _normal _impactnormal) (Collision|BreakHitResult Hit))
  ;[1 1 1] On Audio Collision
  (Audio|OnCollision (Variables|Default|GetCollisionAudioComp) _location (* _impactnormal (Transformation|GetVelocity (Variables|Default|GetPlayerPawn)))))

(fn GetSkeletalMeshFromPawn
  (bind _playerpawn (Variables|Default|GetPlayerPawn))
  (Utilities|IsValid _playerpawn
    (:Is%20Valid
      (Variables|Default|SetPlayerSkeletalMesh (Variables|/Script/Engine.Character|GetMesh _playerpawn))
      (return true))
    (:Is%20Not%20Valid
      (return false))))

(fn ConstructionScript)
