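; Gameplay cue notify that applies the Shadow Bomb invisibility material to a player's current weapon and restores the previous weapon's look.
; Each recurrence reads the pawn from Instigator, current weapon from SourceObject and last weapon from EffectCauser, proceeding only when both weapons change.
; It reapplies cached materials to the last weapon, caches the current one's, then overrides each slot with a dynamic material copying original parameters.
; A pawn-hit handler drives collision audio.
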(event Custom|StoreMaterials
  (bind _currentweapon (Variables|Default|GetCurrentWeapon))
  (bind _weaponmesh (Variables|/Script/FortniteGame.FortWeapon|GetWeaponMesh _currentweapon))
  ;[1 1 1] Store Last weapon materials
  (Utilities|Array|Clear (Variables|Default|GetLastWeaponMaterialInstanceDynamic))
  (for (_ _array_index) (Rendering|Material|GetMaterials _weaponmesh)
    (bind _returnvalue (Fortnite|GetDynamicMaterialInstance _weaponmesh _array_index))
    (Utilities|Array|AddUnique (Variables|Default|GetLastWeaponMaterialInstanceDynamic) _returnvalue)))

(event Custom|ApplyCurrentWeponMaterials
  ;[1 1 1] ApplyCurrentWeponMaterials
  (for (_array_element _array_index) (Rendering|Material|GetMaterials (Variables|/Script/FortniteGame.FortWeapon|GetWeaponMesh (Variables|Default|GetCurrentWeapon)))
    (bind _asmaterial_instance_dynamic (Utilities|Casting|CastToMaterialInstanceDynamic _array_element)
      (:then
        (CallFunction|ApplyDynamicMaterialtoWeaponMesh _asmaterial_instance_dynamic _array_index))
      (:CastFailed
        (bind _returnvalue (Rendering|Material|CreateDynamicMaterialInstance (Variables|Default|GetCurrentWeapon) _array_element "None" "None" :!class /Script/Engine.KismetMaterialLibrary))
        (CallFunction|ApplyDynamicMaterialtoWeaponMesh _returnvalue _array_index)))))

(event Custom|AdjustMaterials
  (bind _lastweapon (Variables|Default|GetLastWeapon))
  ;[1 1 1] Adjust materials
  (for (_array_element _array_index) (Variables|Default|GetLastWeaponMaterialInstanceDynamic)
    (Utilities|IsValid _lastweapon
      (:Is%20Valid
        (Rendering|Material|SetMaterial (Variables|/Script/FortniteGame.FortWeapon|GetWeaponMesh _lastweapon) _array_index _array_element))))
  (CallFunction|StoreMaterials)
  (CallFunction|ApplyCurrentWeponMaterials))

(event EventOnRecurring(Cascade)
  (params (MyTarget :type Actor%20Object%20Reference) (Parameters :type Gameplay%20Cue%20Parameters%20Structure%20%28by%20ref%29) (ParticleComponents :type Array%20of%20Cascade%20Particle%20System%20Component%20Object%20References) (AudioComponents :type Array%20of%20Audio%20Component%20Object%20References) (BurstCameraShakeInstance :type Camera%20Shake%20Base%20Object%20Reference) (BurstDecalInstance :type Decal%20Actor%20Object%20Reference))
  (bind (_normalizedmagnitude _rawmagnitude _effectcontext _matchedtagname _originaltag _aggregatedsourcetags _aggregatedtargettags _location _normal _instigator _effectcauser _sourceobject) (Ability|GameplayCue|BreakGameplayCueParameters Parameters))
  ;[1 1 1] Trigger
  (bind _successful (CallFunction|StorePlayerandPlayerWeaponsfromCueParams _instigator _sourceobject _effectcauser))
  (if _successful
    (CallFunction|AdjustMaterials)))

(event Custom|On%20Target%20Pawn%20Hit%20Event
  (params (SelfActor :type Actor%20Object%20Reference) (OtherActor :type Actor%20Object%20Reference) (NormalImpulse :type Vector) (Hit :type Hit%20Result%20Structure%20%28by%20ref%29))
  (bind (_bblockinghit _binitialoverlap _time _distance _location _impactpoint _normal _impactnormal) (Collision|BreakHitResult Hit))
  ;[1 1 1] On Audio Collision
  (Audio|OnCollision (Variables|Default|GetCollisionAudioComp) _location (* _impactnormal (Transformation|GetVelocity (Variables|Default|GetPlayerPawn)))))

(fn ApplyDynamicMaterialToWeaponMesh
  (params (DefaultMaterial :type Material%20Instance%20Dynamic%20Object%20Reference) (MaterialIndex :type Integer))
  (bind _currentweapon (Variables|Default|GetCurrentWeapon))
  (Utilities|IsValid _currentweapon
    (:Is%20Valid
      (bind _returnvalue (Rendering|Material|CreateDynamicMaterialInstance _currentweapon "/Game/Athena/Items/Consumables/ShadowBomb/FX/MI_Shadow_Bomb_Player_Material_Override.MI_Shadow_Bomb_Player_Material_Override" "None" "None" :!class /Script/Engine.KismetMaterialLibrary))
      (Rendering|Material|CopyMaterialInstanceParameters DefaultMaterial _returnvalue true)
      (Rendering|Material|SetMaterial (Variables|/Script/FortniteGame.FortWeapon|GetWeaponMesh (Variables|Default|GetCurrentWeapon)) MaterialIndex _returnvalue))))

(fn GetSkeletalMeshListFromPawn
  (bind _playerpawn (Variables|Default|GetPlayerPawn))
  (Utilities|IsValid _playerpawn
    (:Is%20Valid
      (Variables|Default|SetPlayerSkeletalMesh (Variables|/Script/Engine.Character|GetMesh _playerpawn))
      (return true))
    (:Is%20Not%20Valid
      (return false))))

(fn StorePlayerAndPlayerWeaponsFromCueParams
  (params (Instigator :type Actor%20Object%20Reference) (CurrentWeapon :type Object%20Reference) (LastWeapon :type Actor%20Object%20Reference))
  (bind _asfort_player_pawn (Utilities|Casting|CastToFortPlayerPawn Instigator)
    (:then
      (Variables|Default|SetPlayerPawn _asfort_player_pawn)
      (bind _asfort_weapon (Utilities|Casting|CastToFortWeapon LastWeapon)
        (:then
          (if (Utilities|NotEqual(Object) _asfort_weapon (Variables|Default|GetLastWeapon))
            (Variables|Default|SetLastWeapon _asfort_weapon)
            (bind _asfort_weapon_2 (Utilities|Casting|CastToFortWeapon CurrentWeapon)
              (:then
                (if (Utilities|NotEqual(Object) _asfort_weapon_2 (Variables|Default|GetCurrentWeapon))
                  (Variables|Default|SetCurrentWeapon _asfort_weapon_2)
                  (return true)
                  (else
                    _)))
              (:CastFailed
                _))
            (return false)
            (else
              _)))
        (:CastFailed
          _)))
    (:CastFailed
      _)))

(fn ConstructionScript)
