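; Looping gameplay cue that pulses a directional trail arrow from the local player toward a tracked Creative objective.
; OnExecute unpacks cue parameters into target, pulse rate and audio flag, then starts a repeating timer.
; Each pulse requires play mode and ends the cue if the target leaves the BreadcrumbTargets register.
; When the camera faces away or distance passes a threshold, it spawns a distance-scaled, device-colored emitter and optional objective-located sound.
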(event Custom|Executed
  (params (PlayAudioTagContainer :type Gameplay%20Tag%20Container%20Structure) (PulseRate :type Float%20%28double-precision%29) (ActorToTrack :type Actor%20Object%20Reference))
  (Variables|Default|SetActor%20to%20Track ActorToTrack)
  (CallFunction|InitCue)
  (Variables|Default|SetPlayAudio (GameplayTags|HasTag PlayAudioTagContainer "(TagName=\"Creative.Components.ObjectiveTrail.PlayAudio\")" true))
  (Utilities|Time|SetTimerbyEvent (delegate Custom|FindClosestObjective) PulseRate true))

(event Custom|FindClosestObjective
  (bind _success (CallFunction|GetClosestObjectiveActor))
  (if _success
    (CallFunction|SpawnEmitter)
    (if (Variables|Default|GetPlayAudio)
      (CallFunction|SpawnAudio))))

(fn GetClosestObjectiveActor
  (bind _actor_to_track (Variables|Default|GetActor%20to%20Track))
  (bind _asfort_player_pawn (Utilities|Casting|CastToFortPlayerPawn (Game|GetPlayerPawn 0)))
  (if (Class|CreativeCommonDeviceFunctionLibrary|IsinPlayMode _asfort_player_pawn)
    (Variables|Default|SetPawn _asfort_player_pawn)
    (CreativeObjectRegistry|GetOrCreateObjectRegister _asfort_player_pawn "(TagName=\"Creative.Register.BreadcrumbTargets\")" false 0 (Variables|Default|GetRegister%20Ref))
    (bind (_bisvalid _register) (Class|FortCreativeIndexedObjectRegisterComponent|GetRegisteratIndex (Variables|Default|GetRegister%20Ref) 0))
    (bind _register_objects (Utilities|Struct|BreakFortObjectArray _register))
    (if (Utilities|Array|ContainsItem _register_objects _actor_to_track)
      (CallFunction|SetNewObjectiveActor _actor_to_track)
      (Variables|Default|SetDistanceBetweenObjectiveAndPlayer (Class|AbilityFunctions|GetDistanceBetweenActors (Variables|Default|GetPawn) (Variables|Default|GetActor%20to%20Track)))
      (return (or (< (Math|Vector|DotProduct (Math|Vector|GetForwardVector (Camera|GetCameraRotation (Variables|/Script/Engine.PlayerController|GetPlayerCameraManager (Variables|Default|GetPlayerController)))) (Math|Vector|Normalize (- (Transformation|GetActorLocation (Variables|Default|GetObjectiveActor)) (Transformation|GetActorLocation (Variables|Default|GetPawn))))) (Variables|Default|GetMax%20Dot%20for%20Arrow)) (> (Variables|Default|GetDistanceBetweenObjectiveAndPlayer) (Math|Float|Square (Variables|Default|GetMin%20Distance%20for%20Arrow)))))
      (else
        ;[1 0 0] There are no more actors to track?  Kill the ability
        (GameplayCueNotify|End%28Recycle%29GameplayCue)
        (return false)))
    (else
      (return false))))

(fn GetDesiredRotation
  (bind _pawn (Variables|Default|GetPawn)
    (:then
      (Variables|Default|GetObjectiveActor
        (:then
          (bind (_returnvalue_roll _returnvalue_pitch _returnvalue_yaw) (Math|Rotator|BreakRotator (Math|Rotator|MakeRotfromX (- (Transformation|GetActorLocation _pawn) (Transformation|GetActorLocation (Variables|Default|GetObjectiveActor))))))
          (return true _returnvalue_roll 9e+01 _returnvalue_yaw))
        (:else
          _)))
    (:else
      _))
  (return false "0, 0, 0"))

(fn InitCue
  (bind _asfort_player_controller_athena (Utilities|Casting|CastToFortPlayerControllerAthena (Fortnite|GetFirstLocalFortPlayerController 0)))
  (Variables|Default|SetPlayerController _asfort_player_controller_athena))

(fn OnBurst
  (params (MyTarget :type Actor%20Object%20Reference) (Parameters :type Gameplay%20Cue%20Parameters%20Structure%20%28by%20ref%29))
  (return false))

(fn OnExecute
  (params (MyTarget :type Actor%20Object%20Reference) (Parameters :type Gameplay%20Cue%20Parameters%20Structure%20%28by%20ref%29))
  (bind (_normalizedmagnitude _rawmagnitude _effectcontext _matchedtagname _originaltag _aggregatedsourcetags _aggregatedtargettags _location _normal _instigator _effectcauser _sourceobject) (Ability|GameplayCue|BreakGameplayCueParameters Parameters))
  (if (Utilities|Equal(Object) (Variables|/Script/Engine.Pawn|GetPlayerState (Game|GetPlayerPawn 0)) _instigator)
    (bind _asactor (Utilities|Casting|CastToActor _sourceobject))
    (CallFunction|Executed _aggregatedsourcetags _normalizedmagnitude _asactor)
    (return true)))

(fn SetNewObjectiveActor
  (params (ObjectiveActor :type Actor%20Object%20Reference))
  (Variables|Default|GetObjectiveActor
    (:then
      (MarkerComponent|UnmarkActoronClient (Class|FortPlayerController|GetMarkerComponent (Variables|Default|GetPlayerController)) (Variables|Default|GetMarker)))
    (:else
      _))
  (Variables|Default|SetObjectiveActor ObjectiveActor)
  (bind _objectiveactor (Variables|Default|GetObjectiveActor)
    (:then
      (bind _color (CreativeDevice|GetDeviceColor(Message) _objectiveactor))
      ;[1 1 1] Setting Color manually to blue in case there is no device color
      (if (not (Math|Color|Equal(LinearColor) :A _color))
        (Variables|Default|SetColor _color)
        (else
          (Variables|Default|SetColor "(R=0.000000,G=0.537697,B=1.000000,A=1.000000)")))
      (bind (_playerbankid _bankinstanceid) (Fortnite|RegisterSoundIndicatorPlacementContext (Variables|Default|GetPlayerController) _objectiveactor))
      (Variables|Default|SetMarker (Utilities|Struct|MakeMarkerID _playerbankid _bankinstanceid)))
    (:else)))

(fn SpawnAudio
  (Audio|PlaySoundatLocation 0 "/Game/Sounds/Fort_UI_Sounds/Tutorial/UI_Tutorial_ProgressBar_Dot_Complete_Cue.UI_Tutorial_ProgressBar_Dot_Complete_Cue" (Transformation|GetActorLocation (Variables|Default|GetObjectiveActor))))

(fn SpawnEmitter
  (bind (_valid _returnvalue) (CallFunction|GetDesiredRotation))
  (if _valid
    (bind _returnvalue_2 (Effects|SpawnEmitterAttached "/Game/Effects/Fort_Effects/Camera_Effects/Effects/P_Screan_DirectionalIndicator_Generic.P_Screan_DirectionalIndicator_Generic" (Variables|/Script/Engine.Character|GetMesh (Variables|Default|GetPawn)) "root" "0.000000,0.000000,100.000000" _returnvalue (Math|Vector|MakeVector (Math|Float|MapRangeClamped (Variables|Default|GetDistanceBetweenObjectiveAndPlayer) 0.0 2.5e+03 1.0 5.0) (Math|Float|MapRangeClamped (Variables|Default|GetDistanceBetweenObjectiveAndPlayer) 0.0 2.5e+03 1.0 5.0) (Math|Float|MapRangeClamped (Variables|Default|GetDistanceBetweenObjectiveAndPlayer) 0.0 2.5e+03 1.0 3.0)) "KeepRelativeOffset" true "None" false))
    (bind _output_get (Variables|Default|SetEmitter _returnvalue_2))
    (Variables|/Script/Engine.SceneComponent|SetbAbsoluteRotation true _output_get)
    ;[1 1 1] Activate here because there is a slight blip that can happen when setting absolute rotation
    (Components|Activation|Activate _output_get false)
    (Effects|Components|ParticleSystem|SetColorParameter _output_get "Tint" (Variables|Default|GetColor))
    (Effects|Components|ParticleSystem|SetEmitterEnable _output_get "Icon" false)
    (Effects|Components|ParticleSystem|SetVectorParameter _output_get "IconLocation" "0, 0, 0")
    (Effects|Components|ParticleSystem|SetEmitterEnable _output_get "Vehicle" false)
    (Effects|Components|ParticleSystem|SetEmitterEnable _output_get "Footsteps" false)
    (Effects|Components|ParticleSystem|SetEmitterEnable _output_get "AttackWaves" false)
    (Effects|Components|ParticleSystem|SetEmitterEnable _output_get "Infected" false)))

(fn ConstructionScript)
