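; Looping gameplay cue that applies a colored glow to a player pawn when the instigating Creative device has the powerup ShowFXOnPlayer tag, otherwise ending itself.
; It reads the instigator's effect data for a direct or team-based color and filters visibility by local view target and team relationship.
; UpdateColor feeds scaled inner and outer colors into a pawn highlight, removed with the cue.
; Customization completion and team-change delegates are used to refresh the color.
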(event EventOnApplication(Cascade)
  (params (MyTarget :type Actor%20Object%20Reference) (Parameters :type Gameplay%20Cue%20Parameters%20Structure%20%28by%20ref%29) (ParticleComponents :type Array%20of%20Cascade%20Particle%20System%20Component%20Object%20References) (AudioComponents :type Array%20of%20Audio%20Component%20Object%20References) (BurstCameraShakeInstance :type Camera%20Shake%20Base%20Object%20Reference) (BurstDecalInstance :type Decal%20Actor%20Object%20Reference))
  (bind (_normalizedmagnitude _rawmagnitude _effectcontext) (Ability|GameplayCue|BreakGameplayCueParameters Parameters))
  (|CallParentFunction MyTarget Parameters ParticleComponents AudioComponents BurstCameraShakeInstance BurstDecalInstance)
  (if (CallFunction|HasShowTag Parameters)
    (bind _asplayer_pawn_athena (Utilities|Casting|CastToPlayerPawn_Athena MyTarget))
    (Variables|Default|SetTarget%20Pawn _asplayer_pawn_athena)
    (CallFunction|ApplyColor (Ability|EffectContext|GetInstigatorActor _effectcontext))
    (else
      (GameplayCueNotify|End%28Recycle%29GameplayCue))))

(event Custom|ApplyColor
  (params (Instigating%20Actor :type Actor%20Object%20Reference))
  (Variables|Default|SetInstigating%20Actor Instigating%20Actor)
  (bind _playerstate (Variables|/Script/Engine.Pawn|GetPlayerState (Variables|Default|GetTarget%20Pawn))
    (:then
      (bind _returnvalue (VolumeManager|GetMinigameforPlayerState _playerstate))
      (Variables|Default|SetMinigame _returnvalue)
      (CallFunction|UpdateColor))
    (:else)))

(event EventOnRemoval(Cascade)
  (params (MyTarget :type Actor%20Object%20Reference) (Parameters :type Gameplay%20Cue%20Parameters%20Structure%20%28by%20ref%29) (ParticleComponents :type Array%20of%20Cascade%20Particle%20System%20Component%20Object%20References) (AudioComponents :type Array%20of%20Audio%20Component%20Object%20References) (BurstCameraShakeInstance :type Camera%20Shake%20Base%20Object%20Reference) (BurstDecalInstance :type Decal%20Actor%20Object%20Reference))
  (CallFunction|BindOrUnbindTeamUpdate false)
  (PawnHighlight|RemovePawnHighlightFromSource (Variables|Default|GetTarget%20Pawn) self))

(event EventOnLoopingStart(Cascade)
  (params (MyTarget :type Actor%20Object%20Reference) (Parameters :type Gameplay%20Cue%20Parameters%20Structure%20%28by%20ref%29) (ParticleComponents :type Array%20of%20Cascade%20Particle%20System%20Component%20Object%20References) (AudioComponents :type Array%20of%20Audio%20Component%20Object%20References))
  (bind _target_pawn (Variables|Default|GetTarget%20Pawn))
  (|CallParentFunction MyTarget Parameters ParticleComponents AudioComponents)
  (if (CallFunction|HasShowTag Parameters)
    (bind _target_pawn_2 (Variables|Default|GetTarget%20Pawn)
      (:then
        (Customization|UnbindEventfromOnFinishedCharacterCustomization _target_pawn_2 (EventDispatchers|CreateEvent :!event Custom|On%20Customization%20Complete)) ; Switching Pawn. Unbind from previous pawn first if it exists.
      )
      (:else
        _))
    (bind _asplayer_pawn_athena (Utilities|Casting|CastToPlayerPawn_Athena MyTarget))
    (bind _output_get (Variables|Default|SetTarget%20Pawn _asplayer_pawn_athena))
    (if (Customization|IsCharacterCustomizationLoadingCompleted _output_get)
      (CallFunction|ApplyColor _target_pawn)
      (else
        _))
    (Customization|BindEventtoOnFinishedCharacterCustomization _target_pawn (EventDispatchers|CreateEvent :!event Custom|On%20Customization%20Complete)) ; Bind even if we already have a customization as we want to \nreapply if customization changes.
    (else
      (GameplayCueNotify|End%28Recycle%29GameplayCue))))

(event Custom|On%20Customization%20Complete
  (params (Pawn :type Fort%20Player%20Pawn%20Object%20Reference))
  ;[1 1 1] When our pawn customization changes, reapply the color
  (CallFunction|ApplyColor Pawn))

(fn BindorUnbindTeamUpdate
  (params (Bind :type Boolean))
  (bind _outputdelegate (EventDispatchers|CreateEvent :!event Update%20Color))
  (bind _target_pawn (Variables|Default|GetTarget%20Pawn)
    (:then
      (bind _asathena_player_controller (Utilities|Casting|CastToAthena_PlayerController (Fortnite|GetFirstLocalFortPlayerController _target_pawn)))
      (if (Variables|GetBind)
        (FortPlayerControllerAthena|BindEventtoOnLocalPlayerChangedTeams _asathena_player_controller _outputdelegate)
        (bind _minigame (Variables|Default|GetMinigame)
          (:then
            (Minigame|BindEventtoMinigameTeamConfigUpdated _minigame _outputdelegate))
          (:else))
        (else
          (FortPlayerControllerAthena|UnbindEventfromOnLocalPlayerChangedTeams _asathena_player_controller _outputdelegate)
          (bind _minigame_2 (Variables|Default|GetMinigame)
            (:then
              (Minigame|UnbindEventfromMinigameTeamConfigUpdated _minigame_2 _outputdelegate))
            (:else)))))
    (:else)))

(fn GetColor
  (bind _effect_color (CreativeEffects|GetCreativeEffectData(Message) (Variables|Default|GetInstigating%20Actor)))
  (bind (_effect_color_color_13_d2d2deb24e760961e0d6ab88dc603692 _effect_color_settype_5_3a2997d544a0618ec312feaed95187b8) (Utilities|Struct|BreakFCreativeEffectDirectColor _effect_color))
  (bind _show_effect (CallFunction|ShouldShowEffectonPawn))
  (if _show_effect
    (switch Utilities|FlowControl|Switch|SwitchonECreativeColorSetType _effect_color_settype_5_3a2997d544a0618ec312feaed95187b8
      (:NewEnumerator0
        (return false 0))
      (:NewEnumerator1
        (return true _effect_color_color_13_d2d2deb24e760961e0d6ab88dc603692))
      (:NewEnumerator2
        _)
      (:NewEnumerator3
        _))
    (CallFunction|BindOrUnbindTeamUpdate true)
    (bind _result (ArrayMacros|SafeArrayGet (Class|GetClassDefaults "/Game/Creative/Sets/MilitaryBase/Props/CreativeTeamColors.CreativeTeamColors_C") (Teams|GetDynamicTeamColor (Variables|Default|GetTarget%20Pawn) (Variables|Default|GetMinigame) _effect_color_settype_5_3a2997d544a0618ec312feaed95187b8 (Variables|Default|GetMinigame)))
      (:%20%20
        (return true _result)))
    (else
      _)))

(fn HasShowTag
  (params (Parameters :type Gameplay%20Cue%20Parameters%20Structure))
  (bind (_normalizedmagnitude _rawmagnitude _effectcontext) (Ability|GameplayCue|BreakGameplayCueParameters Parameters))
  (bind _returnvalue (GameplayTags|HasMatchingGameplayTag(Message) (Ability|EffectContext|GetInstigatorActor _effectcontext) "(TagName=\"Creative.UserOptions.PowerupDevice.ShowFXOnPlayer\")"))
  (return _returnvalue))

(fn SetFresnelColor
  (bind (_override_color _result) (CallFunction|GetColor))
  (if _override_color
    (bind _output_get (LocalVariables|Setlocal%20Color _result))
    (Variables|DirectSet|SetOuterColor (Math|Color|LinearColor*Float _output_get 2.0))
    (Variables|DirectSet|SetInnerColor (Math|Color|LinearColor*Float (LocalVariables|Getlocal%20Color) (Variables|Default|GetInnerGlowScale)))))

(fn ShouldShowEffectOnPawn
  (bind _effect_color (CreativeEffects|GetCreativeEffectData(Message) (Variables|Default|GetInstigating%20Actor)))
  (bind (_effect_color_color_13_d2d2deb24e760961e0d6ab88dc603692 _effect_color_settype_5_3a2997d544a0618ec312feaed95187b8 _effect_color_visibletoowningplayer_8_e546f2a4426b8ab99ee118a3125ac217 _effect_color_teamrelationshipvisibility_11_3bb5265d4d611d2bb5799eb5a6e26aa7) (Utilities|Struct|BreakFCreativeEffectDirectColor _effect_color))
  (return (and (or (Utilities|NotEqual(Object) (Variables|Default|GetTarget%20Pawn) (Pawn|GetViewTarget (Fortnite|GetFirstLocalFortPlayerController (Variables|Default|GetInstigating%20Actor)))) (Utilities|Enum|Equal(Enum) _effect_color_visibletoowningplayer_8_e546f2a4426b8ab99ee118a3125ac217 "NewEnumerator0")) (Math|Boolean|ORBoolean (and (Team|OnSameTeam (Variables|Default|GetTarget%20Pawn) (Pawn|GetViewTarget (Fortnite|GetFirstLocalFortPlayerController (Variables|Default|GetInstigating%20Actor)))) (Utilities|Enum|Equal(Enum) _effect_color_teamrelationshipvisibility_11_3bb5265d4d611d2bb5799eb5a6e26aa7 "NewEnumerator0")) (and (not (Team|OnSameTeam (Variables|Default|GetTarget%20Pawn) (Pawn|GetViewTarget (Fortnite|GetFirstLocalFortPlayerController (Variables|Default|GetInstigating%20Actor))))) (Utilities|Enum|Equal(Enum) _effect_color_teamrelationshipvisibility_11_3bb5265d4d611d2bb5799eb5a6e26aa7 "NewEnumerator2")) (Utilities|Enum|Equal(Enum) _effect_color_teamrelationshipvisibility_11_3bb5265d4d611d2bb5799eb5a6e26aa7 "NewEnumerator1")))))

(fn UpdateColor
  (CallFunction|SetFresnelColor)
  (bind _show_effect (CallFunction|ShouldShowEffectonPawn))
  (if _show_effect
    (PawnHighlight|HighlightPawnFromSource (Variables|Default|GetTarget%20Pawn) self (Utilities|Struct|MakePawnHighlight (Variables|DirectSet|GetGlow%20Priority) (Variables|DirectSet|GetInnerColor) (Variables|DirectSet|GetOuterColor) (Variables|DirectSet|GetHitGlowFresnelBrightness) (Variables|DirectSet|GetHitGlowFresnelExponent) (Variables|DirectSet|GetUse%20Pulse)) false)))

(fn ConstructionScript)
