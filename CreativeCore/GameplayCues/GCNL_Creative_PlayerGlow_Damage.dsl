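; Looping gameplay cue notify for a Creative player damage glow, a child of a glow cue that deferred start and removal go to.
; It overrides parent hooks to supply a fixed purple color, scaled into outer and inner fresnel colors.
; Tag and pawn checks always pass, avoiding tags because the device may not exist on the client.
; Remaining duration comes from an active gameplay effect matching the GE tag.
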(event EventOnLoopingStart(Cascade)
  (params (MyTarget :type Actor%20Object%20Reference) (Parameters :type Gameplay%20Cue%20Parameters%20Structure%20%28by%20ref%29) (ParticleComponents :type Array%20of%20Cascade%20Particle%20System%20Component%20Object%20References) (AudioComponents :type Array%20of%20Audio%20Component%20Object%20References))
  (|CallParentFunction MyTarget Parameters ParticleComponents AudioComponents))

(event EventOnRemoval(Cascade)
  (params (MyTarget :type Actor%20Object%20Reference) (Parameters :type Gameplay%20Cue%20Parameters%20Structure%20%28by%20ref%29) (ParticleComponents :type Array%20of%20Cascade%20Particle%20System%20Component%20Object%20References) (AudioComponents :type Array%20of%20Audio%20Component%20Object%20References) (BurstCameraShakeInstance :type Camera%20Shake%20Base%20Object%20Reference) (BurstDecalInstance :type Decal%20Actor%20Object%20Reference))
  (|CallParentFunction MyTarget Parameters ParticleComponents AudioComponents BurstCameraShakeInstance BurstDecalInstance))

(fn GetColor
  (return true "(R=0.180392,G=0.031373,B=0.282353,A=1.000000)"))

(fn GetRemainingDuration
  (params (Actor :type Actor%20Object%20Reference))
  (Utilities|IsValid Actor
    (:Is%20Valid
      (bind _returnvalue (GameplayEffects|GetActiveEffectswithAllTags (Ability|GetAbilitySystemComponent Actor) (Variables|Default|GetGE%20Tag)))
      (bind _result (ArrayMacros|SafeArrayGet _returnvalue)
        (:%20%20
          (bind _returnvalue_2 (Ability|GameplayEffect|GetActiveGameplayEffectRemainingDuration Actor _result))
          (return _returnvalue_2))))
    (:Is%20Not%20Valid
      (return 0.0))))

(fn HasEffectTag
  (params (InputPin :type Gameplay%20Cue%20Parameters%20Structure%20%28by%20ref%29))
  (return true))

(fn HasShowTag
  (params (Parameters :type Gameplay%20Cue%20Parameters%20Structure))
  ;[1 1 1] don't use tags because the device could not exist in the client
  (return true))

(fn SetFresnelColor
  (bind _output_get (LocalVariables|Setlocal%20Color "(R=0.180392,G=0.031373,B=0.282353,A=1.000000)"))
  (Variables|DirectSet|SetOuterColor (Math|Color|LinearColor*Float _output_get 2.0))
  (Variables|DirectSet|SetInnerColor (Math|Color|LinearColor*Float (LocalVariables|Getlocal%20Color) (Variables|Default|GetInnerGlowScale))))

(fn ShouldShowEffectOnPawn
  (return true))

(fn ConstructionScript
  (|CallParentFunction))
