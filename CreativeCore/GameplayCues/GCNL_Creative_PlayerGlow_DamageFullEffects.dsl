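; Looping gameplay cue for the Creative player glow damage effect, giving the target pawn a purple glow plus audio feedback.
; While active, it binds the pawn's melee swing and ranged fire events to play an attached fire sound, unbinding on removal.
; It reads the tagged GameplayEffect's remaining duration and uses a retriggerable delay to play warning audio before expiry.
; Show checks always return true, bypassing tags because the device cannot exist on clients.
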(event EventOnLoopingStart(Cascade)
  (params (MyTarget :type Actor%20Object%20Reference) (Parameters :type Gameplay%20Cue%20Parameters%20Structure%20%28by%20ref%29) (ParticleComponents :type Array%20of%20Cascade%20Particle%20System%20Component%20Object%20References) (AudioComponents :type Array%20of%20Audio%20Component%20Object%20References))
  (bind _target_pawn (Variables|Default|GetTarget%20Pawn))
  (bind _returnvalue (Utilities|Time|GetGameTimeinSeconds _target_pawn))
  (|CallParentFunction MyTarget Parameters ParticleComponents AudioComponents)
  (CallFunction|BindOrUnbindFireEvents true)
  (bind _returnvalue_2 (CallFunction|GetRemainingDuration _target_pawn))
  (if (> (+ _returnvalue _returnvalue_2) (Variables|Default|GetEnd%20Time))
    (Variables|Default|SetEnd%20Time (+ _returnvalue _returnvalue_2))
    (Utilities|FlowControl|RetriggerableDelay (Variables|Default|GetTarget%20Pawn) (Math|Float|Max(Float) (- _returnvalue_2 (Variables|Default|GetWarning%20Time))))
    (Audio|SpawnSoundAttached (Variables|Default|GetWarning%20Audio) (Variables|/Script/Engine.Actor|GetRootComponent (Variables|Default|GetTarget%20Pawn)))))

(event Custom|Weapon%20Fired
  (params (GameplayAbility :type Fort%20Gameplay%20Ability%20Object%20Reference) (AbilityTags :type Gameplay%20Tag%20Container%20Structure))
  (CallFunction|PlayWeaponFireFX))

(event Custom|Play%20Weapon%20Fire%20FX
  (bind _target_pawn (Variables|Default|GetTarget%20Pawn))
  (Utilities|IsValid _target_pawn
    (:Is%20Valid
      (Audio|SpawnSoundAttached (Variables|Default|GetFire%20Audio) (Variables|/Script/Engine.Actor|GetRootComponent _target_pawn)))))

(event EventOnRemoval(Cascade)
  (params (MyTarget :type Actor%20Object%20Reference) (Parameters :type Gameplay%20Cue%20Parameters%20Structure%20%28by%20ref%29) (ParticleComponents :type Array%20of%20Cascade%20Particle%20System%20Component%20Object%20References) (AudioComponents :type Array%20of%20Audio%20Component%20Object%20References) (BurstCameraShakeInstance :type Camera%20Shake%20Base%20Object%20Reference) (BurstDecalInstance :type Decal%20Actor%20Object%20Reference))
  (|CallParentFunction MyTarget Parameters ParticleComponents AudioComponents BurstCameraShakeInstance BurstDecalInstance)
  (CallFunction|BindOrUnbindFireEvents false))

(fn BindorUnbindFireEvents
  (params (Bind :type Boolean))
  (bind _outputdelegate (EventDispatchers|CreateEvent :!event Play%20Weapon%20Fire%20FX))
  (bind _outputdelegate_2 (EventDispatchers|CreateEvent :!event Weapon%20Fired))
  (bind _target_pawn (Variables|Default|GetTarget%20Pawn))
  (bind _target_pawn_2 (Variables|Default|GetTarget%20Pawn))
  (Variables|Default|GetTarget%20Pawn
    (:then
      (if Bind
        (Default|BindEventtoSwingLeft _target_pawn _outputdelegate)
        (Default|BindEventtoSwingLeft2 _target_pawn _outputdelegate)
        (Default|BindEventtoSwingRight _target_pawn _outputdelegate)
        (Default|BindEventtoSwingRight2 _target_pawn _outputdelegate)
        (EventDispatchers|BindEventtoOnFiringRangedWeapon _target_pawn _outputdelegate_2)
        (else
          (Default|UnbindEventfromSwingLeft _target_pawn_2 _outputdelegate)
          (Default|UnbindEventfromSwingLeft2 _target_pawn_2 _outputdelegate)
          (Default|UnbindEventfromSwingRight _target_pawn_2 _outputdelegate)
          (Default|UnbindEventfromSwingRight2 _target_pawn_2 _outputdelegate)
          (EventDispatchers|UnbindEventfromOnFiringRangedWeapon _target_pawn_2 _outputdelegate_2))))
    (:else)))

(fn GetColor
  (return true "(R=0.180392,G=0.031373,B=0.282353,A=1.000000)"))

(fn GetRemainingDuration
  (params (Actor :type Actor%20Object%20Reference))
  (Utilities|IsValid Actor
    (:Is%20Valid
      (bind _returnvalue (GameplayEffects|GetActiveEffectswithAllTags (Ability|GetAbilitySystemComponent Actor) (Variables|Default|GetGE%20Tag)))
      (bind _result (ArrayMacros|SafeArrayGet _returnvalue)
        (:%20%20
          (bind _returnvalue_2 (Ability|GameplayEffect|GetActiveGameplayEffectRemainingDuration Actor _result))
          (return _returnvalue_2))))
    (:Is%20Not%20Valid
      (return 0.0))))

(fn HasShowTag
  (params (Parameters :type Gameplay%20Cue%20Parameters%20Structure))
  ;[1 1 1] Don't use tags because the device cannot exist on the client.
  (return true))

(fn SetFresnelColor
  (bind _output_get (LocalVariables|Setlocal%20Color "(R=0.180392,G=0.031373,B=0.282353,A=1.000000)"))
  (Variables|DirectSet|SetOuterColor (Math|Color|LinearColor*Float _output_get 2.0))
  (Variables|DirectSet|SetInnerColor (Math|Color|LinearColor*Float (LocalVariables|Getlocal%20Color) (Variables|Default|GetInnerGlowScale))))

(fn ShouldShowEffectOnPawn
  (return true))

(fn ConstructionScript
  (|CallParentFunction))
