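; Looping gameplay cue that plays a Creative player level-up particle and sound effect on the target pawn, skipped on mobile.
; On application it caches instigator, pawn and minigame, then shows and attaches the effect if the instigator's effect data allows visibility for the local view target's ownership and team.
; Particle color is either direct or team-derived from CreativeTeamColors.
; Team-derived color binds UpdateColor to local team changes, unbound on removal.
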(event EventOnApplication(Cascade)
  (params (MyTarget :type Actor%20Object%20Reference) (Parameters :type Gameplay%20Cue%20Parameters%20Structure%20%28by%20ref%29) (ParticleComponents :type Array%20of%20Cascade%20Particle%20System%20Component%20Object%20References) (AudioComponents :type Array%20of%20Audio%20Component%20Object%20References) (BurstCameraShakeInstance :type Camera%20Shake%20Base%20Object%20Reference) (BurstDecalInstance :type Decal%20Actor%20Object%20Reference))
  (bind (_normalizedmagnitude _rawmagnitude _effectcontext) (Ability|GameplayCue|BreakGameplayCueParameters Parameters))
  (|CallParentFunction MyTarget Parameters ParticleComponents AudioComponents BurstCameraShakeInstance BurstDecalInstance)
  (if (not (Fortnite|IsMobilePlatform))
    (Variables|Default|SetInstigating%20Actor (Ability|EffectContext|GetInstigatorActor _effectcontext))
    (bind _asplayer_pawn_athena (Utilities|Casting|CastToPlayerPawn_Athena MyTarget))
    (bind _output_get (Variables|Default|SetTarget%20Pawn _asplayer_pawn_athena))
    (bind _playerstate (Variables|/Script/Engine.Pawn|GetPlayerState _output_get)
      (:then
        (bind _returnvalue (VolumeManager|GetMinigameforPlayerState _playerstate))
        (Variables|Default|SetMinigame _returnvalue)
        (CallFunction|PlayEffect))
      (:else))))

(event EventOnRemoval(Cascade)
  (params (MyTarget :type Actor%20Object%20Reference) (Parameters :type Gameplay%20Cue%20Parameters%20Structure%20%28by%20ref%29) (ParticleComponents :type Array%20of%20Cascade%20Particle%20System%20Component%20Object%20References) (AudioComponents :type Array%20of%20Audio%20Component%20Object%20References) (BurstCameraShakeInstance :type Camera%20Shake%20Base%20Object%20Reference) (BurstDecalInstance :type Decal%20Actor%20Object%20Reference))
  (|CallParentFunction MyTarget Parameters ParticleComponents AudioComponents BurstCameraShakeInstance BurstDecalInstance)
  (CallFunction|BindOrUnbindTeamUpdate false))

(fn BindorUnbindTeamUpdate
  (params (Bind :type Boolean))
  (bind _outputdelegate (EventDispatchers|CreateEvent :!event Update%20Color))
  (bind _target_pawn (Variables|Default|GetTarget%20Pawn)
    (:then
      (bind _asathena_player_controller (Utilities|Casting|CastToAthena_PlayerController (Fortnite|GetFirstLocalFortPlayerController _target_pawn)))
      (if (Variables|GetBind)
        (FortPlayerControllerAthena|BindEventtoOnLocalPlayerChangedTeams _asathena_player_controller _outputdelegate)
        (else
          (FortPlayerControllerAthena|UnbindEventfromOnLocalPlayerChangedTeams _asathena_player_controller _outputdelegate))))
    (:else)))

(fn GetColor
  (bind _effect_color (CreativeEffects|GetCreativeEffectData(Message) (Variables|Default|GetInstigating%20Actor)))
  (bind (_effect_color_color_13_d2d2deb24e760961e0d6ab88dc603692 _effect_color_settype_5_3a2997d544a0618ec312feaed95187b8) (Utilities|Struct|BreakFCreativeEffectDirectColor _effect_color))
  (bind _show_effect (CallFunction|ShouldShowEffectonPawn))
  (if _show_effect
    (switch Utilities|FlowControl|Switch|SwitchonECreativeColorSetType _effect_color_settype_5_3a2997d544a0618ec312feaed95187b8
      (:NewEnumerator0
        (return 0))
      (:NewEnumerator1
        (return _effect_color_color_13_d2d2deb24e760961e0d6ab88dc603692))
      (:NewEnumerator2
        _)
      (:NewEnumerator3
        _))
    (CallFunction|BindOrUnbindTeamUpdate true)
    (bind _result (ArrayMacros|SafeArrayGet (Class|GetClassDefaults "/Game/Creative/Sets/MilitaryBase/Props/CreativeTeamColors.CreativeTeamColors_C") (Teams|GetDynamicTeamColor (Variables|Default|GetTarget%20Pawn) (Variables|Default|GetMinigame) _effect_color_settype_5_3a2997d544a0618ec312feaed95187b8 (Variables|Default|GetMinigame)))
      (:%20%20
        (return _result)))
    (else
      _)))

(fn PlayEffect
  (bind _vfx (Variables|Default|GetVFX))
  (bind _show_effect (CallFunction|ShouldShowEffectonPawn))
  (if _show_effect
    (Components|Activation|Activate _vfx true)
    (CallFunction|UpdateColor)
    (Audio|Components|Audio|Play (Variables|Default|GetAudio) 0.0)
    (Transformation|AttachActorToActor (Variables|Default|GetTarget%20Pawn) "Root" "SnapToTarget" "SnapToTarget" "KeepWorld")
    (else
      (Components|Activation|Deactivate _vfx))))

(fn ShouldShowEffectOnPawn
  (bind _effect_color (CreativeEffects|GetCreativeEffectData(Message) (Variables|Default|GetInstigating%20Actor)))
  (bind (_effect_color_color_13_d2d2deb24e760961e0d6ab88dc603692 _effect_color_settype_5_3a2997d544a0618ec312feaed95187b8 _effect_color_visibletoowningplayer_8_e546f2a4426b8ab99ee118a3125ac217 _effect_color_teamrelationshipvisibility_11_3bb5265d4d611d2bb5799eb5a6e26aa7) (Utilities|Struct|BreakFCreativeEffectDirectColor _effect_color))
  (return (and (or (Utilities|NotEqual(Object) (Variables|Default|GetTarget%20Pawn) (Pawn|GetViewTarget (Fortnite|GetFirstLocalFortPlayerController (Variables|Default|GetInstigating%20Actor)))) (Utilities|Enum|Equal(Enum) _effect_color_visibletoowningplayer_8_e546f2a4426b8ab99ee118a3125ac217 "NewEnumerator0")) (Math|Boolean|ORBoolean (and (Team|OnSameTeam (Variables|Default|GetTarget%20Pawn) (Pawn|GetViewTarget (Fortnite|GetFirstLocalFortPlayerController (Variables|Default|GetInstigating%20Actor)))) (Utilities|Enum|Equal(Enum) _effect_color_teamrelationshipvisibility_11_3bb5265d4d611d2bb5799eb5a6e26aa7 "NewEnumerator0")) (and (not (Team|OnSameTeam (Variables|Default|GetTarget%20Pawn) (Pawn|GetViewTarget (Fortnite|GetFirstLocalFortPlayerController (Variables|Default|GetInstigating%20Actor))))) (Utilities|Enum|Equal(Enum) _effect_color_teamrelationshipvisibility_11_3bb5265d4d611d2bb5799eb5a6e26aa7 "NewEnumerator2")) (Utilities|Enum|Equal(Enum) _effect_color_teamrelationshipvisibility_11_3bb5265d4d611d2bb5799eb5a6e26aa7 "NewEnumerator1")))))

(fn UpdateColor
  ;[1 0 0 0.25] Stub for color switching
  (bind _color (CallFunction|GetColor))
  (Effects|Components|ParticleSystem|SetColorParameter (Variables|Default|GetVFX) "Color" _color))

(fn ConstructionScript
  (|CallParentFunction))
