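; Creative gameplay cue that supplies a player outline's color and visibility from the instigating actor's effect data.
; On application it caches the instigator and the target's minigame, then calls the parent cue.
; GetColor resolves a fixed, direct or dynamic team color, gated by owner and team-relationship visibility against the local view target.
; It binds Update Color to local team changes, and removal unbinds it.
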(event EventOnApplication(EitherVFX-preferthisone)
  (params (MyTarget :type Actor%20Object%20Reference) (Parameters :type Gameplay%20Cue%20Parameters%20Structure%20%28by%20ref%29) (ParticleComponents :type Array%20of%20FXSystem%20Component%20Object%20References) (AudioComponents :type Array%20of%20Audio%20Component%20Object%20References) (BurstCameraShakeInstance :type Camera%20Shake%20Base%20Object%20Reference) (BurstDecalInstance :type Decal%20Actor%20Object%20Reference))
  (bind (_normalizedmagnitude _rawmagnitude _effectcontext) (Ability|GameplayCue|BreakGameplayCueParameters Parameters))
  (Utilities|FlowControl|Sequence
    (:then_0
      (Variables|Default|SetInstigating%20Actor (Ability|EffectContext|GetInstigatorActor _effectcontext))
      (bind _asplayer_pawn_athena (Utilities|Casting|CastToPlayerPawn_Athena MyTarget))
      (bind _playerstate (Variables|/Script/Engine.Pawn|GetPlayerState _asplayer_pawn_athena)
        (:then
          (bind _returnvalue (VolumeManager|GetMinigameforPlayerState _playerstate))
          (Variables|Default|SetMinigame _returnvalue))
        (:else)))
    (:then_1
      (|CallParentFunction MyTarget Parameters ParticleComponents AudioComponents BurstCameraShakeInstance BurstDecalInstance)
      (|CallParentFunction MyTarget Parameters ParticleComponents AudioComponents))))

(event EventOnLoopingStart(EitherVFX-preferthisversion)
  (params (MyTarget :type Actor%20Object%20Reference) (Parameters :type Gameplay%20Cue%20Parameters%20Structure%20%28by%20ref%29) (ParticleComponents :type Array%20of%20FXSystem%20Component%20Object%20References) (AudioComponents :type Array%20of%20Audio%20Component%20Object%20References)))

(event EventOnRemoval(EitherVFX-preferthisversion)
  (params (MyTarget :type Actor%20Object%20Reference) (Parameters :type Gameplay%20Cue%20Parameters%20Structure%20%28by%20ref%29) (ParticleComponents :type Array%20of%20FXSystem%20Component%20Object%20References) (AudioComponents :type Array%20of%20Audio%20Component%20Object%20References) (BurstCameraShakeInstance :type Camera%20Shake%20Base%20Object%20Reference) (BurstDecalInstance :type Decal%20Actor%20Object%20Reference))
  (|CallParentFunction MyTarget Parameters ParticleComponents AudioComponents BurstCameraShakeInstance BurstDecalInstance)
  (CallFunction|BindOrUnbindTeamUpdate false))

(fn BindorUnbindTeamUpdate
  (params (Bind :type Boolean))
  (bind _outputdelegate (EventDispatchers|CreateEvent :!event Update%20Color))
  (bind _target_pawn (Variables|Default|GetTarget%20Pawn)
    (:then
      (bind _asathena_player_controller (Utilities|Casting|CastToAthena_PlayerController (Fortnite|GetFirstLocalFortPlayerController _target_pawn)))
      (if (Variables|GetBind)
        (FortPlayerControllerAthena|BindEventtoOnLocalPlayerChangedTeams _asathena_player_controller _outputdelegate)
        (else
          (FortPlayerControllerAthena|UnbindEventfromOnLocalPlayerChangedTeams _asathena_player_controller _outputdelegate))))
    (:else)))

(fn GetColor
  (bind _effect_color (CreativeEffects|GetCreativeEffectData(Message) (Variables|Default|GetInstigating%20Actor)))
  (bind (_effect_color_color_13_d2d2deb24e760961e0d6ab88dc603692 _effect_color_settype_5_3a2997d544a0618ec312feaed95187b8) (Utilities|Struct|BreakFCreativeEffectDirectColor _effect_color))
  (bind _show_effect (CallFunction|ShouldShowEffectonPawn))
  (if _show_effect
    (switch Utilities|FlowControl|Switch|SwitchonECreativeColorSetType _effect_color_settype_5_3a2997d544a0618ec312feaed95187b8
      (:NewEnumerator0
        (return "(R=1.000000,G=0.000000,B=0.988789,A=1.000000)"))
      (:NewEnumerator1
        (CallFunction|BindOrUnbindTeamUpdate true)
        (return _effect_color_color_13_d2d2deb24e760961e0d6ab88dc603692))
      (:NewEnumerator2
        _)
      (:NewEnumerator3
        _))
    (CallFunction|BindOrUnbindTeamUpdate true)
    (bind _result (ArrayMacros|SafeArrayGet (Class|GetClassDefaults "/Game/Creative/Sets/MilitaryBase/Props/CreativeTeamColors.CreativeTeamColors_C") (Teams|GetDynamicTeamColor (Variables|Default|GetTarget%20Pawn) (Variables|Default|GetMinigame) _effect_color_settype_5_3a2997d544a0618ec312feaed95187b8 (Variables|Default|GetMinigame)))
      (:%20%20
        (return _result)))
    (else
      _)))

(fn ShouldShowEffectOnPawn
  (bind _instigating_actor (Variables|Default|GetInstigating%20Actor)
    (:then
      (bind _effect_color (CreativeEffects|GetCreativeEffectData(Message) _instigating_actor))
      (bind (_effect_color_color_13_d2d2deb24e760961e0d6ab88dc603692 _effect_color_settype_5_3a2997d544a0618ec312feaed95187b8 _effect_color_visibletoowningplayer_8_e546f2a4426b8ab99ee118a3125ac217 _effect_color_teamrelationshipvisibility_11_3bb5265d4d611d2bb5799eb5a6e26aa7) (Utilities|Struct|BreakFCreativeEffectDirectColor _effect_color))
      (bind _target_pawn (Variables|Default|GetTarget%20Pawn)
        (:then
          (return (and (or (Utilities|NotEqual(Object) _target_pawn (Pawn|GetViewTarget (Fortnite|GetFirstLocalFortPlayerController _instigating_actor))) (Utilities|Enum|Equal(Enum) _effect_color_visibletoowningplayer_8_e546f2a4426b8ab99ee118a3125ac217 "NewEnumerator0")) (Math|Boolean|ORBoolean (and (Team|OnSameTeam _target_pawn (Pawn|GetViewTarget (Fortnite|GetFirstLocalFortPlayerController _instigating_actor))) (Utilities|Enum|Equal(Enum) _effect_color_teamrelationshipvisibility_11_3bb5265d4d611d2bb5799eb5a6e26aa7 "NewEnumerator0")) (and (not (Team|OnSameTeam _target_pawn (Pawn|GetViewTarget (Fortnite|GetFirstLocalFortPlayerController _instigating_actor)))) (Utilities|Enum|Equal(Enum) _effect_color_teamrelationshipvisibility_11_3bb5265d4d611d2bb5799eb5a6e26aa7 "NewEnumerator2")) (Utilities|Enum|Equal(Enum) _effect_color_teamrelationshipvisibility_11_3bb5265d4d611d2bb5799eb5a6e26aa7 "NewEnumerator1")))))
        (:else
          _)))
    (:else
      _))
  (return false))

(fn ConstructionScript
  (|CallParentFunction))
