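; Looping gameplay cue that shows a Creative powerup sparks effect on the affected player pawn, filtered for the local viewer.
; On application it stores instigator and target, then activates and attaches the VFX or deactivates it.
; Visibility comes from the instigator's Creative effect data, checking owning-player and team-relationship settings against the local view target.
; It re-evaluates when the local player changes teams, binding that event on application and unbinding on removal.
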(event Custom|UpdateVisibility
  (bind _vfx (Variables|Default|GetVFX))
  (bind _show_effect (Default|ShouldShowEffectonPawn))
  (if _show_effect
    (Components|Activation|Activate _vfx true)
    (Transformation|AttachActorToActor (Variables|Default|GetTarget%20Pawn) "Root" "SnapToTarget" "SnapToTarget" "KeepWorld")
    (else
      (Components|Activation|Deactivate _vfx))))

(event EventOnApplication(Cascade)
  (params (MyTarget :type Actor%20Object%20Reference) (Parameters :type Gameplay%20Cue%20Parameters%20Structure%20%28by%20ref%29) (ParticleComponents :type Array%20of%20Cascade%20Particle%20System%20Component%20Object%20References) (AudioComponents :type Array%20of%20Audio%20Component%20Object%20References) (BurstCameraShakeInstance :type Camera%20Shake%20Base%20Object%20Reference) (BurstDecalInstance :type Decal%20Actor%20Object%20Reference))
  (bind (_normalizedmagnitude _rawmagnitude _effectcontext) (Ability|GameplayCue|BreakGameplayCueParameters Parameters))
  (bind _vfx (Variables|Default|GetVFX))
  (|CallParentFunction MyTarget Parameters ParticleComponents AudioComponents BurstCameraShakeInstance BurstDecalInstance)
  (Utilities|FlowControl|Sequence
    (:then_0
      (Variables|Default|SetInstigating%20Actor (Ability|EffectContext|GetInstigatorActor _effectcontext))
      (bind _asplayer_pawn_athena (Utilities|Casting|CastToPlayerPawn_Athena MyTarget))
      (Variables|Default|SetTarget%20Pawn _asplayer_pawn_athena))
    (:then_1
      (Default|BindOrUnbindTeamUpdate true)
      (bind _show_effect (Default|ShouldShowEffectonPawn))
      (if _show_effect
        (Components|Activation|Activate _vfx true)
        (Transformation|AttachActorToActor (Variables|Default|GetTarget%20Pawn) "Root" "SnapToTarget" "SnapToTarget" "KeepWorld")
        (else
          (Components|Activation|Deactivate _vfx))))))

(event EventOnRemoval(Cascade)
  (params (MyTarget :type Actor%20Object%20Reference) (Parameters :type Gameplay%20Cue%20Parameters%20Structure%20%28by%20ref%29) (ParticleComponents :type Array%20of%20Cascade%20Particle%20System%20Component%20Object%20References) (AudioComponents :type Array%20of%20Audio%20Component%20Object%20References) (BurstCameraShakeInstance :type Camera%20Shake%20Base%20Object%20Reference) (BurstDecalInstance :type Decal%20Actor%20Object%20Reference))
  (|CallParentFunction MyTarget Parameters ParticleComponents AudioComponents BurstCameraShakeInstance BurstDecalInstance)
  (Default|BindOrUnbindTeamUpdate false))

(fn BindorUnbindTeamUpdate
  (params (Bind :type Boolean))
  (bind _outputdelegate (EventDispatchers|CreateEvent :!event UpdateVisibility))
  (bind _target_pawn (Variables|Default|GetTarget%20Pawn)
    (:then
      (bind _asathena_player_controller (Utilities|Casting|CastToAthena_PlayerController (Fortnite|GetFirstLocalFortPlayerController _target_pawn)))
      (if (Variables|GetBind)
        (FortPlayerControllerAthena|BindEventtoOnLocalPlayerChangedTeams _asathena_player_controller _outputdelegate)
        (else
          (FortPlayerControllerAthena|UnbindEventfromOnLocalPlayerChangedTeams _asathena_player_controller _outputdelegate))))
    (:else)))

(fn ShouldShowEffectOnPawn
  (bind _effect_color (CreativeEffects|GetCreativeEffectData(Message) (Variables|Default|GetInstigating%20Actor)))
  (bind (_effect_color_color_13_d2d2deb24e760961e0d6ab88dc603692 _effect_color_settype_5_3a2997d544a0618ec312feaed95187b8 _effect_color_visibletoowningplayer_8_e546f2a4426b8ab99ee118a3125ac217 _effect_color_teamrelationshipvisibility_11_3bb5265d4d611d2bb5799eb5a6e26aa7) (Utilities|Struct|BreakFCreativeEffectDirectColor _effect_color))
  (return (and (or (Utilities|NotEqual(Object) (Variables|Default|GetTarget%20Pawn) (Pawn|GetViewTarget (Fortnite|GetFirstLocalFortPlayerController (Variables|Default|GetInstigating%20Actor)))) (Utilities|Enum|Equal(Enum) _effect_color_visibletoowningplayer_8_e546f2a4426b8ab99ee118a3125ac217 "NewEnumerator0")) (Math|Boolean|ORBoolean (and (Team|OnSameTeam (Variables|Default|GetTarget%20Pawn) (Pawn|GetViewTarget (Fortnite|GetFirstLocalFortPlayerController (Variables|Default|GetInstigating%20Actor)))) (Utilities|Enum|Equal(Enum) _effect_color_teamrelationshipvisibility_11_3bb5265d4d611d2bb5799eb5a6e26aa7 "NewEnumerator0")) (and (not (Team|OnSameTeam (Variables|Default|GetTarget%20Pawn) (Pawn|GetViewTarget (Fortnite|GetFirstLocalFortPlayerController (Variables|Default|GetInstigating%20Actor))))) (Utilities|Enum|Equal(Enum) _effect_color_teamrelationshipvisibility_11_3bb5265d4d611d2bb5799eb5a6e26aa7 "NewEnumerator2")) (Utilities|Enum|Equal(Enum) _effect_color_teamrelationshipvisibility_11_3bb5265d4d611d2bb5799eb5a6e26aa7 "NewEnumerator1")))))

(fn ConstructionScript)
