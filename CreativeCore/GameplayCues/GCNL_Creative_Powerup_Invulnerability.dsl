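; Looping gameplay cue for the Creative invulnerability powerup that fades a highlight on the player pawn and plays GhostMode enter and exit sounds.
; On start it registers the pawn's Invulnerable Highlight with itself as source, then a timeline updates that highlight through HighlightLerp, scaling colors and fresnel up from defaults.
; On removal the same timeline reverses, and removes the highlight when it finishes backward.
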(event EventOnLoopingStart(Cascade)
  (params (MyTarget :type Actor%20Object%20Reference) (Parameters :type Gameplay%20Cue%20Parameters%20Structure%20%28by%20ref%29) (ParticleComponents :type Array%20of%20Cascade%20Particle%20System%20Component%20Object%20References) (AudioComponents :type Array%20of%20Audio%20Component%20Object%20References))
  (bind _asplayer_pawn_athena (Utilities|Casting|CastToPlayerPawn_Athena MyTarget))
  (Variables|Default|SetTarget%20Player _asplayer_pawn_athena)
  (CallFunction|FadeMaterialsIn)
  (Audio|PlaySoundatLocation MyTarget "/Game/Sounds/Creative/Gadgets/GhostMode/GhostMode_Enter_Cue.GhostMode_Enter_Cue" (Transformation|GetActorLocation)))

(event EventOnRemoval(Cascade)
  (params (MyTarget :type Actor%20Object%20Reference) (Parameters :type Gameplay%20Cue%20Parameters%20Structure%20%28by%20ref%29) (ParticleComponents :type Array%20of%20Cascade%20Particle%20System%20Component%20Object%20References) (AudioComponents :type Array%20of%20Audio%20Component%20Object%20References) (BurstCameraShakeInstance :type Camera%20Shake%20Base%20Object%20Reference) (BurstDecalInstance :type Decal%20Actor%20Object%20Reference))
  (CallFunction|FadeMaterialsOut)
  (Audio|PlaySoundatLocation MyTarget "/Game/Sounds/Creative/Gadgets/GhostMode/GhostMode_Exit_Cue.GhostMode_Exit_Cue" (Transformation|GetActorLocation)))

(event Custom|FadeMaterialsIn
  (bind _target_player (Variables|Default|GetTarget%20Player)
    (:then
      (PawnHighlight|HighlightPawnFromSource _target_player self (Variables|/Game/Athena/PlayerPawn_Athena.PlayerPawn_Athena_C|GetInvulnerable%20Highlight _target_player) false) ; Setting a full highlight with self as Source to the current one.\nThis isn't seen by the player unless this only becomes current \nafter the timeline has ended.
      (bind (_direction _lerp) (Timeline|FadeMaterialsTL PlayFromStart)
        (:Update
          (bind _target_player_2 (Variables|Default|GetTarget%20Player)
            (:then
              (PawnHighlight|HighlightPawnFromSource (Variables|Default|GetTarget%20Player) self (CallFunction|HighlightLerp (Variables|/Game/Athena/PlayerPawn_Athena.PlayerPawn_Athena_C|GetInvulnerable%20Highlight _target_player_2) _lerp) true) ; If current, lerp in the colors
            )
            (:else)))
        (:Finished
          (if (Utilities|Enum|Equal(Enum) _direction "Backward")
            (bind _target_player_3 (Variables|Default|GetTarget%20Player)
              (:then
                (PawnHighlight|RemovePawnHighlightFromSource _target_player_3 self))
              (:else))))))
    (:else)))

(event Custom|FadeMaterialsOut
  (Timeline|FadeMaterialsTL Reverse))

(fn HighlightLerp
  (params (In%20Highlight_Priority_28_E2E1B5344846E187B9C11B863A7F0698 :type Float%20%28single-precision%29) (In%20Highlight_Inner_21_4CC2801147EA190DE16F59B34F36853E :type Linear%20Color%20Structure) (In%20Highlight_Outer_22_5A1D7D0543D303E8B54B66A7F7BD2E2E :type Linear%20Color%20Structure) (In%20Highlight_FresnelBrightness_23_52B0F96447FF640F47DF2895B0602E92 :type Float%20%28single-precision%29) (In%20Highlight_FresnelExponent_24_B427CF0C441AA37ED49833BF7579DE6D :type Float%20%28single-precision%29) (In%20Highlight_UsesPulse_25_E29229F64E540F0617E4C4987AD77605 :type Float%20%28single-precision%29) (Alpha :type Float%20%28double-precision%29))
  (return In%20Highlight_Priority_28_E2E1B5344846E187B9C11B863A7F0698 (Math|Color|Lerp(LinearColor) :B In%20Highlight_Inner_21_4CC2801147EA190DE16F59B34F36853E :Alpha (Variables|GetAlpha)) (Math|Color|Lerp(LinearColor) :B In%20Highlight_Outer_22_5A1D7D0543D303E8B54B66A7F7BD2E2E :Alpha (Variables|GetAlpha)) (Math|Float|Lerp 0.0 In%20Highlight_FresnelBrightness_23_52B0F96447FF640F47DF2895B0602E92 (Variables|GetAlpha)) (Math|Float|Lerp 0.0 In%20Highlight_FresnelExponent_24_B427CF0C441AA37ED49833BF7579DE6D (Variables|GetAlpha)) In%20Highlight_UsesPulse_25_E29229F64E540F0617E4C4987AD77605))

(fn ConstructionScript)
