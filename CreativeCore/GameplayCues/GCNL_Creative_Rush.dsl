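; Looping gameplay cue for the Creative Rush effect, adding speed-increase particles and sounds to a player pawn.
; On application it attaches Cascade emitters to both foot sockets and the body, then plays the increase sound on the mesh.
; Only the foot emitters are stored; removal deactivates them with raised time dilation, clears the array and plays the stop sound.
; The body emitter is untracked, so removal never explicitly stops it.
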(event EventOnApplication(Cascade)
  (params (MyTarget :type Actor%20Object%20Reference) (Parameters :type Gameplay%20Cue%20Parameters%20Structure%20%28by%20ref%29) (ParticleComponents :type Array%20of%20Cascade%20Particle%20System%20Component%20Object%20References) (AudioComponents :type Array%20of%20Audio%20Component%20Object%20References) (BurstCameraShakeInstance :type Camera%20Shake%20Base%20Object%20Reference) (BurstDecalInstance :type Decal%20Actor%20Object%20Reference))
  (bind _asfort_player_pawn (Utilities|Casting|CastToFortPlayerPawn MyTarget))
  (Variables|Default|SetMesh (Variables|/Script/Engine.Character|GetMesh _asfort_player_pawn))
  (bind _returnvalue (Effects|SpawnEmitterAttached :EmitterTemplate "/Game/Items/Traps/Particles/P_SpeedIncrease_Feet.P_SpeedIncrease_Feet" :AttachToComponent (Variables|/Script/Engine.Character|GetMesh _asfort_player_pawn) :AttachPointName "FX_foot_lSocket" :Scale "1.000000,1.000000,1.000000" :LocationType "KeepRelativeOffset" :bAutoDestroy true :PoolingMethod "AutoRelease" :bAutoActivate true))
  (Utilities|Array|Add (Variables|Default|GetPSComponents) _returnvalue)
  (bind _returnvalue_2 (Effects|SpawnEmitterAttached :EmitterTemplate "/Game/Items/Traps/Particles/P_SpeedIncrease_Feet.P_SpeedIncrease_Feet" :AttachToComponent (Variables|/Script/Engine.Character|GetMesh _asfort_player_pawn) :AttachPointName "FX_foot_rSocket" :Scale "1.000000,1.000000,1.000000" :LocationType "KeepRelativeOffset" :bAutoDestroy true :PoolingMethod "AutoRelease" :bAutoActivate true))
  (Utilities|Array|Add (Variables|Default|GetPSComponents) _returnvalue_2)
  (Effects|SpawnEmitterAttached :EmitterTemplate "/Game/Items/Traps/Particles/P_SpeedIncrease_Body.P_SpeedIncrease_Body" :AttachToComponent (Variables|/Script/Engine.Character|GetMesh _asfort_player_pawn) :AttachPointName "None" :Scale "1.000000,1.000000,1.000000" :LocationType "KeepRelativeOffset" :bAutoDestroy true :PoolingMethod "AutoRelease" :bAutoActivate true)
  (Audio|SpawnSoundAttached (Variables|Default|GetIncrease%20Speed%20Sound) (Variables|Default|GetMesh)))

(event EventOnRemoval(Cascade)
  (params (MyTarget :type Actor%20Object%20Reference) (Parameters :type Gameplay%20Cue%20Parameters%20Structure%20%28by%20ref%29) (ParticleComponents :type Array%20of%20Cascade%20Particle%20System%20Component%20Object%20References) (AudioComponents :type Array%20of%20Audio%20Component%20Object%20References) (BurstCameraShakeInstance :type Camera%20Shake%20Base%20Object%20Reference) (BurstDecalInstance :type Decal%20Actor%20Object%20Reference))
  (bind _asfort_player_pawn (Utilities|Casting|CastToFortPlayerPawn MyTarget))
  (Variables|Default|SetMesh (Variables|/Script/Engine.Character|GetMesh _asfort_player_pawn))
  (for _array_element (Variables|Default|GetPSComponents)
    (Utilities|IsValid _array_element
      (:Is%20Valid
        (Components|Activation|Deactivate _array_element)
        (Variables|/Script/CascadeRuntime.ParticleSystemComponent|SetCustomTimeDilation 5.0 _array_element))))
  (Utilities|Array|Clear (Variables|Default|GetPSComponents))
  (Audio|SpawnSoundAttached (Variables|Default|GetIncrease%20Stop%20Sound) (Variables|Default|GetMesh)))

(fn ConstructionScript)
