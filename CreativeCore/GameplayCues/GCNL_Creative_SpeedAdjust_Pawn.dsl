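; Looping gameplay cue for Creative speed adjustment that shows speed-up or slow-down particles and sound on the target pawn.
; On becoming relevant it clears prior effects, then converts the effect level to a speed-setting enum to choose increase, decrease or cleanup.
; Emitters attach to foot sockets on player pawns or four paw sockets on other pawns, plus the body.
; Tracked fast and slow emitters are deactivated on removal.
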(event Custom|Active
  (params (Parameters :type Gameplay%20Cue%20Parameters%20Structure%20%28by%20ref%29) (InputPin :type Actor%20Object%20Reference))
  (bind (_normalizedmagnitude _rawmagnitude _effectcontext _matchedtagname _originaltag _aggregatedsourcetags _aggregatedtargettags _location _normal _instigator _effectcauser _sourceobject _physicalmaterial _gameplayeffectlevel) (Ability|GameplayCue|BreakGameplayCueParameters Parameters))
  ;[1 1 1] Any re-application needs to be cleaned up first
  (CallFunction|CleanupValidPlayer)
  (bind _asfort_player_pawn (Utilities|Casting|CastToFortPlayerPawn InputPin)
    (:then
      (Variables|Default|SetTargetPawn _asfort_player_pawn))
    (:CastFailed
      (bind _asfort_pawn (Utilities|Casting|CastToFortPawn InputPin))
      (Variables|Default|SetTargetPawn _asfort_pawn)))
  (switch Utilities|FlowControl|Switch|SwitchonToySpeedSettings (CallFunction|LeveltoToySpeedSettings _gameplayeffectlevel)
    (:NewEnumerator0
      _)
    (:NewEnumerator1
      _)
    (:NewEnumerator2
      _)
    (:NewEnumerator3
      (CallFunction|CleanupValidPlayer))
    (:NewEnumerator4
      (CallFunction|SpawnandAttachSpeedParticleSystem (Variables|Default|GetP_SpeedIncrease) (Variables|Default|GetP_SpeedIncreaseBody) true)
      (CallFunction|PlaySpeedAdjustmentSound true))
    (:NewEnumerator5
      _)
    (:NewEnumerator6
      _))
  (CallFunction|SpawnandAttachSpeedParticleSystem (Variables|Default|GetP_SpeedDecrease) (Variables|Default|GetP_SpeedDecreaseBody))
  (CallFunction|PlaySpeedAdjustmentSound false))

(event EventOnRemoval(Cascade)
  (params (MyTarget :type Actor%20Object%20Reference) (Parameters :type Gameplay%20Cue%20Parameters%20Structure%20%28by%20ref%29) (ParticleComponents :type Array%20of%20Cascade%20Particle%20System%20Component%20Object%20References) (AudioComponents :type Array%20of%20Audio%20Component%20Object%20References) (BurstCameraShakeInstance :type Camera%20Shake%20Base%20Object%20Reference) (BurstDecalInstance :type Decal%20Actor%20Object%20Reference))
  (Utilities|ValidPlayer MyTarget
    (:true
      (CallFunction|CleanupValidPlayer))))

(fn AttachEmitterToSocket
  (params (EmitterTemplate :type Particle%20System%20Object%20Reference) (AttachPointName :type Name) (bIncreaseSpeed :type Boolean))
  (bind _targetpawn (Variables|Default|GetTargetPawn))
  (Utilities|Casting|CastToFortPlayerPawn _targetpawn
    (:then
      (LocalVariables|SetTargetSK (CallFunction|GetPlayerAttachmentComponent)))
    (:CastFailed
      (LocalVariables|SetTargetSK (Variables|/Script/Engine.Character|GetMesh _targetpawn))))
  (bind _returnvalue (Effects|SpawnEmitterAttached :EmitterTemplate EmitterTemplate :AttachToComponent (LocalVariables|GetTargetSK) :AttachPointName AttachPointName :Scale "1.000000,1.000000,1.000000" :LocationType "KeepRelativeOffset" :bAutoDestroy true :PoolingMethod "AutoRelease" :bAutoActivate true))
  (if bIncreaseSpeed
    (Utilities|Array|Add (Variables|Default|GetFastParticleSystems) _returnvalue)
    (else
      (Utilities|Array|Add (Variables|Default|GetSlowParticleSystems) _returnvalue))))

(fn CleanupValidPlayer
  (CallFunction|DeactivateandClearParticleSystems (Variables|Default|GetFastParticleSystems))
  (CallFunction|DeactivateandClearParticleSystems (Variables|Default|GetSlowParticleSystems))
  (Variables|Default|SetTargetPawn 0))

(fn DeactivateAndClearParticleSystems
  (params (Array :type Array%20of%20Cascade%20Particle%20System%20Component%20Object%20References))
  (for _array_element Array
    (Components|Activation|Deactivate _array_element))
  (Utilities|Array|Clear Array))

(fn GetPlayerAttachmentComponent
  (return (Variables|/Script/Engine.Character|GetMesh (Variables|Default|GetTargetPawn))))

(fn LevelToToySpeedSettings
  (params (InInt :type Integer))
  (return (Utilities|Enum|BytetoEnumToySpeedSettings (Math|Conversions|ToByte(Integer) InInt))))

(fn PlaySpeedAdjustmentSound
  (params (IncreasedSpeed :type Boolean))
  (Audio|SpawnSoundAttached (select IncreasedSpeed (Variables|Default|GetIncrease%20Speed%20Sound) (Variables|Default|GetDecrease%20Speed%20Sound)) (CallFunction|GetPlayerAttachmentComponent)))

(fn SpawnAndAttachSpeedParticleSystem
  (params (EmitterTemplate :type Particle%20System%20Object%20Reference) (EmitterTemplateBody :type Particle%20System%20Object%20Reference) (bIncreaseSpeed :type Boolean))
  (Utilities|Casting|CastToFortPlayerPawn (Variables|Default|GetTargetPawn)
    (:then
      (CallFunction|AttachEmittertoSocket EmitterTemplate "FX_foot_lSocket" bIncreaseSpeed)
      (CallFunction|AttachEmittertoSocket EmitterTemplate "FX_foot_rSocket" bIncreaseSpeed)
      (CallFunction|AttachEmittertoSocket EmitterTemplateBody "None" bIncreaseSpeed))
    (:CastFailed
      (CallFunction|AttachEmittertoSocket EmitterTemplate "FX_RearPaw_L" bIncreaseSpeed)
      (CallFunction|AttachEmittertoSocket EmitterTemplate "FX_RearPaw_R" bIncreaseSpeed)
      (CallFunction|AttachEmittertoSocket EmitterTemplate "FX_FrontPaw_L" bIncreaseSpeed)
      (CallFunction|AttachEmittertoSocket EmitterTemplate "FX_FrontPaw_R" bIncreaseSpeed)
      (CallFunction|AttachEmittertoSocket EmitterTemplateBody "None" bIncreaseSpeed))))

(fn ConstructionScript)

(fn OnBecomeRelevant
  (params (MyTarget :type Actor%20Object%20Reference) (Parameters :type Gameplay%20Cue%20Parameters%20Structure%20%28by%20ref%29))
  (CallFunction|Active Parameters MyTarget)
  (return true))
