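; Looping gameplay cue Blueprint, named for a speed adjustment effect on vehicles in Creative mode.
; It holds no speed or effect logic: BeginPlay only prints a "Hello Vehicle" debug string.
; The construction script is empty, so the asset appears to be a placeholder.
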(event EventBeginPlay
  (Development|PrintString 0 "Hello Vehicle"))

(fn ConstructionScript)
