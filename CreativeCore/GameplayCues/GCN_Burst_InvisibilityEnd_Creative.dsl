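; Burst gameplay cue that, by its name, plays when Creative invisibility ends and makes the player's held dual-wield weapon visible again.
; After running the parent burst, it casts the target to an Athena player pawn and checks its current weapon.
; Dual ranged weapons get their left-hand mesh shown and a left-weapon re-enable flag set.
; Otherwise, dual-wield pickaxes get both offhand and main meshes shown; other weapons are left untouched here.
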(fn OnBurst(Cascade)
  (params (MyTarget :type Actor%20Object%20Reference) (Parameters :type Gameplay%20Cue%20Parameters%20Structure%20%28by%20ref%29) (ParticleComponents :type Array%20of%20Cascade%20Particle%20System%20Component%20Object%20References) (AudioComponents :type Array%20of%20Audio%20Component%20Object%20References) (BurstCameraShakeInstance :type Camera%20Shake%20Base%20Object%20Reference) (BurstDecalInstance :type Decal%20Actor%20Object%20Reference))
  (|CallParentFunction MyTarget Parameters ParticleComponents AudioComponents BurstCameraShakeInstance BurstDecalInstance)
  (bind _asfort_player_pawn_athena (Utilities|Casting|CastToFortPlayerPawnAthena MyTarget)
    (:then
      (Utilities|IsValid (Variables|/Script/FortniteGame.FortPawn|GetCurrentWeapon _asfort_player_pawn_athena)
        (:Is%20Valid
          (bind _asb_ranged_dual_generic_athena (Utilities|Casting|CastToB_Ranged_Dual_Generic_Athena (Variables|/Script/FortniteGame.FortPawn|GetCurrentWeapon _asfort_player_pawn_athena))
            (:then
              (Variables|/Game/Weapons/Blueprints/B_Ranged_Dual_Generic_Athena.B_Ranged_Dual_Generic_Athena_C|SetbShouldReEnableLeftWeapVis true _asb_ranged_dual_generic_athena)
              (Rendering|SetVisibility (Class|FortWeaponRangedDual|GetLeftHandWeaponMesh _asb_ranged_dual_generic_athena) true true))
            (:CastFailed
              (bind _asfort_weapon_pickaxe_dual_wield_athena (Utilities|Casting|CastToFortWeaponPickaxeDualWieldAthena (Variables|/Script/FortniteGame.FortPawn|GetCurrentWeapon _asfort_player_pawn_athena))
                (:then
                  (Utilities|IsValid (Variables|/Script/FortniteGame.FortWeaponPickaxeDualWieldAthena|GetWeaponMeshOffhand _asfort_weapon_pickaxe_dual_wield_athena)
                    (:Is%20Valid
                      (Rendering|SetVisibility (Variables|/Script/FortniteGame.FortWeaponPickaxeDualWieldAthena|GetWeaponMeshOffhand _asfort_weapon_pickaxe_dual_wield_athena) true true))
                    (:Is%20Not%20Valid
                      _))
                  (Utilities|IsValid (Visual|GetWeaponMesh (Variables|/Script/FortniteGame.FortPawn|GetCurrentWeapon _asfort_player_pawn_athena))
                    (:Is%20Valid
                      (Rendering|SetVisibility (Visual|GetWeaponMesh (Variables|/Script/FortniteGame.FortPawn|GetCurrentWeapon _asfort_player_pawn_athena)) true true))
                    (:Is%20Not%20Valid
                      _))
                  (return))
                (:CastFailed
                  _)))))
        (:Is%20Not%20Valid
          _)))
    (:CastFailed
      _)))
