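; Looping Cascade gameplay cue for Creative's near-end invisibility effect, using Shadow Bomb sound.
; On start it drives mesh, particle and global mist parameters by viewer type, tracks team, and hides remote players' dual-wield weapons on equip.
; Local players get a repeating expiry sound and flash, while another timer feeds vertical speed to audio.
; Setup repeats after character customization instead of using delays; removal clears warning timers, mist and weapon binding.
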(event Custom|BeginExpireTell
  ;[1 1 1] Handle ability expire warning - local player only
  (Actor|Tick|SetActorTickEnabled false)
  (bind _returnvalue (Utilities|Time|SetTimerbyEvent (delegate Custom|ExpireWarningTick) (Variables|Default|GetExpirationSoundPeriod) true))
  (Variables|Default|SetExpirationSoundTimer _returnvalue))

(event Custom|ExpireWarningTick
  (bind _expiretellopacity (Variables|Default|GetExpireTellOpacity))
  ;[1 1 1] Handle ability expire warning - local player only
  (Audio|SpawnSoundAttached "/FNGameplayCues/Athena/Sounds/Items/Shadow_Bomb/Sahdow_Bomb_Expire_Tell_Cue.Sahdow_Bomb_Expire_Tell_Cue" (Variables|/Script/Engine.Actor|GetRootComponent (Variables|Default|GetTargetPlayer)))
  (for _array_element (Variables|Default|GetPlayerSkeletalMeshes)
    (Rendering|Material|SetScalarParameterValueonMaterials _array_element "FlashStartTime" (Utilities|Time|GetGameTimeinSeconds 0))
    (if (not (Math|Boolean|ORBoolean (Replay|GetIsPlayingReplay 0) (Broadcast|GetIsBroadcastClient 0) (Pawn|IsLocallyControlled (Variables|Default|GetTargetPlayer))))
      (Effects|Components|ParticleSystem|SetFloatParameter (Utilities|Array|Get(acopy) (Variables|Default|GetAttachedFX)) "MovementOpacity" _expiretellopacity)
      (if (> (Utilities|Array|Length (Variables|Default|GetPlayerSkeletalMeshes)) 0)
        (Rendering|Material|SetScalarParameterValueonMaterials _array_element "MovementOpacity" _expiretellopacity)
        (else
          _))
      (Variables|Default|SetExpireTellOpacity (Math|Float|Clamp(Float) (+ (Variables|Default|GetExpireTellOpacity) 0.15))))))

(event Custom|ClearExpireTell
  ;[1 1 1] Handle ability expire warning - local player only
  (Utilities|Time|ClearandInvalidateTimerbyHandle 0 (Variables|Default|GetExpireTellDelayTimer))
  (Utilities|Time|ClearandInvalidateTimerbyHandle 0 (Variables|Default|GetExpirationSoundTimer)))

(event EventOnLoopingStart(Cascade)
  (params (MyTarget :type Actor%20Object%20Reference) (Parameters :type Gameplay%20Cue%20Parameters%20Structure%20%28by%20ref%29) (ParticleComponents :type Array%20of%20Cascade%20Particle%20System%20Component%20Object%20References) (AudioComponents :type Array%20of%20Audio%20Component%20Object%20References))
  (bind _playerskeletalmeshes (Variables|Default|GetPlayerSkeletalMeshes))
  (bind _targetplayer (Variables|Default|GetTargetPlayer))
  (Utilities|FlowControl|Sequence
    (:then_0
      (Utilities|Time|SetTimerbyEvent (EventDispatchers|CreateEvent :!event Custom|Audio%20Tick) 0.1 true)
      (Variables|Default|SetVisbilityLevel 1.0)
      (Variables|Default|SetFriendlyPlayer false)
      (Variables|Default|SetTickENabled true)
      (Components|Timeline|ReversefromEnd (Variables|Default|GetFadeInOut))
      (Variables|Default|SetWalkVisMultiplier (ScalableFloat|GetValueatLevel (Variables|Default|GetDataWalkVisMult)))
      (Variables|Default|SetSprintVisMultiplier (ScalableFloat|GetValueatLevel (Variables|Default|GetDataSprintVisMult)))
      (Variables|Default|SetMinFriendlyOpacity (ScalableFloat|GetValueatLevel (Variables|Default|GetDataFriendlyVisMult)))
      (Variables|Default|SetExpireTellOpacity 0.0))
    (:then_1
      (Variables|Default|SetAttachedFX ParticleComponents)
      (Effects|Components|ParticleSystem|SetFloatParameter (Utilities|Array|Get(acopy) (Variables|Default|GetAttachedFX)) "MovementOpacity" 0.0)
      (if (> (Utilities|Array|Length _playerskeletalmeshes) 0)
        (for _array_element _playerskeletalmeshes
          (Rendering|Material|SetScalarParameterValueonMaterials _array_element "MovementOpacity" 0.0))))
    (:then_2
      (bind _asfort_player_pawn (Utilities|Casting|CastToFortPlayerPawn MyTarget))
      (bind _output_get (Variables|Default|SetTargetPlayer _asfort_player_pawn))
      ;[1 0.01042 0.01042] DON'T USE DELAYS: This fixes the Xbox One issue where the first time you throw the smoke bomb, you are invisible instead of misty.
      ; We're binding to when the customization is done then do the things.
      ; We're also just passing all the way through in case the bind doesn't get hit on time.
      (Customization|BindEventtoOnFinishedCharacterCustomization _output_get (delegate Custom|OnFinishedCharacterCustomization_Bind))
      (Utilities|FlowControl|Sequence
        (:then_0
          (if (Pawn|IsLocallyControlled (Variables|Default|GetTargetPlayer))
            ;[1 1 1] Handle ability expire warning - local player only
            (CallFunction|ClearExpireTell)
            (CallFunction|BeginExpireTell)))
        (:then_1
          (if (Math|Boolean|ORBoolean (Pawn|IsLocallyControlled (Variables|Default|GetTargetPlayer)) (Replay|GetIsPlayingReplay 0) (Broadcast|GetIsBroadcastClient 0))
            (Rendering|Material|SetScalarParameterValue 0 "/Game/Packages/Fortress_SharedMaterials/GlobalMaterialParameters/FortniteMaterialParameters.FortniteMaterialParameters" "SpookyMistMode" (ScalableFloat|GetValueatLevel (Variables|Default|GetDataShadowPlaneOpacity)) :!class /Script/Engine.KismetMaterialLibrary) ; Reveal other players using ghost rocks
            (Variables|Default|SetFriendlyPlayer true)
            (Effects|Components|ParticleSystem|SetFloatParameter (Utilities|Array|Get(acopy) (Variables|Default|GetAttachedFX) 1) "Friendly" 1.0)
            (CallFunction|GetSkeletalMeshes)
            (for _array_element_2 (Variables|Default|GetPlayerSkeletalMeshes)
              (Rendering|Material|SetScalarParameterValueonMaterials _array_element_2 "MovementOpacity" 1.0))
            (else
              (CallFunction|GetSkeletalMeshes)
              (for _array_element_3 (Fortnite|GetLocalFortPlayerControllers 0)
                (bind _targetplayer_2 (Variables|Default|GetTargetPlayer)
                  (:then
                    (Variables|Default|SetFriendlyPlayer (Team|OnSameTeam (Variables|/Script/Engine.Pawn|GetPlayerState _targetplayer_2) (Variables|/Script/Engine.Controller|GetPlayerState _array_element_3)))
                    (if (not (Pawn|IsLocallyControlled (Variables|Default|GetTargetPlayer)))
                      (FortPawn|AssignOnWeaponEquippedDelegate _targetplayer (delegate Custom|OnWeaponEquippedDelegate_Event_0))
                      (CallFunction|SetWeaponVisibility (Variables|/Script/FortniteGame.FortPawn|GetCurrentWeapon _targetplayer))))
                  (:else))))))))))

(event Custom|OnFinishedCharacterCustomization_Bind
  (params (Pawn :type Fort%20Player%20Pawn%20Object%20Reference))
  (bind _targetplayer (Variables|Default|GetTargetPlayer))
  ;[1 0.01042 0.01042] DON'T USE DELAYS: This fixes the Xbox One issue where the first time you throw the smoke bomb, you are invisible instead of misty.
  ; We're binding to when the customization is done then do the things.
  ; We're also just passing all the way through in case the bind doesn't get hit on time.
  (Customization|UnbindEventfromOnFinishedCharacterCustomization Pawn (delegate Custom|OnFinishedCharacterCustomization_Bind))
  (Utilities|FlowControl|Sequence
    (:then_0
      (if (Pawn|IsLocallyControlled (Variables|Default|GetTargetPlayer))
        ;[1 1 1] Handle ability expire warning - local player only
        (CallFunction|ClearExpireTell)
        (CallFunction|BeginExpireTell)))
    (:then_1
      (if (Math|Boolean|ORBoolean (Pawn|IsLocallyControlled (Variables|Default|GetTargetPlayer)) (Replay|GetIsPlayingReplay 0) (Broadcast|GetIsBroadcastClient 0))
        (Rendering|Material|SetScalarParameterValue 0 "/Game/Packages/Fortress_SharedMaterials/GlobalMaterialParameters/FortniteMaterialParameters.FortniteMaterialParameters" "SpookyMistMode" (ScalableFloat|GetValueatLevel (Variables|Default|GetDataShadowPlaneOpacity)) :!class /Script/Engine.KismetMaterialLibrary) ; Reveal other players using ghost rocks
        (Variables|Default|SetFriendlyPlayer true)
        (Effects|Components|ParticleSystem|SetFloatParameter (Utilities|Array|Get(acopy) (Variables|Default|GetAttachedFX) 1) "Friendly" 1.0)
        (CallFunction|GetSkeletalMeshes)
        (for _array_element (Variables|Default|GetPlayerSkeletalMeshes)
          (Rendering|Material|SetScalarParameterValueonMaterials _array_element "MovementOpacity" 1.0))
        (else
          (CallFunction|GetSkeletalMeshes)
          (for _array_element_2 (Fortnite|GetLocalFortPlayerControllers 0)
            (bind _targetplayer_2 (Variables|Default|GetTargetPlayer)
              (:then
                (Variables|Default|SetFriendlyPlayer (Team|OnSameTeam (Variables|/Script/Engine.Pawn|GetPlayerState _targetplayer_2) (Variables|/Script/Engine.Controller|GetPlayerState _array_element_2)))
                (if (not (Pawn|IsLocallyControlled (Variables|Default|GetTargetPlayer)))
                  (FortPawn|AssignOnWeaponEquippedDelegate _targetplayer (delegate Custom|OnWeaponEquippedDelegate_Event_0))
                  (CallFunction|SetWeaponVisibility (Variables|/Script/FortniteGame.FortPawn|GetCurrentWeapon _targetplayer))))
              (:else))))))))

(event Custom|UpdateTeamVisbility
  (bind _targetplayer (Variables|Default|GetTargetPlayer))
  (CallFunction|GetSkeletalMeshes)
  (for _array_element (Fortnite|GetLocalFortPlayerControllers 0)
    (bind _targetplayer_2 (Variables|Default|GetTargetPlayer)
      (:then
        (Variables|Default|SetFriendlyPlayer (Team|OnSameTeam (Variables|/Script/Engine.Pawn|GetPlayerState _targetplayer_2) (Variables|/Script/Engine.Controller|GetPlayerState _array_element)))
        (if (not (Pawn|IsLocallyControlled (Variables|Default|GetTargetPlayer)))
          (FortPawn|AssignOnWeaponEquippedDelegate _targetplayer (delegate Custom|OnWeaponEquippedDelegate_Event_0))
          (CallFunction|SetWeaponVisibility (Variables|/Script/FortniteGame.FortPawn|GetCurrentWeapon _targetplayer))))
      (:else))))

(event Custom|Audio%20Tick
  (bind _targetplayer (Variables|Default|GetTargetPlayer)
    (:then
      (bind _output_get (Variables|Default|SetVelocityInterp (Math|Interpolation|FInterpTo (Variables|Default|GetVelocityInterp) (Math|Float|MapRangeClamped (Math|Float|Absolute(Float) (.z (Transformation|GetVelocity _targetplayer))) 6e+02 1e+03 0.0 1.0) (Utilities|Time|GetWorldDeltaSeconds 0) 5.0)))
      (Audio|Parameter|SetFloatParameter (Utilities|Array|Get(acopy) (stale EventOnLoopingStart(Cascade) AudioComponents)) "Velocity" _output_get :!class /Script/AudioExtensions.AudioParameterControllerInterface))
    (:else)))

(event Custom|OnWeaponEquippedDelegate_Event_0
  (params (NewWeapon :type Fort%20Weapon%20Object%20Reference) (PrevWeapon :type Fort%20Weapon%20Object%20Reference))
  (CallFunction|SetWeaponVisibility PrevWeapon true)
  (CallFunction|SetWeaponVisibility NewWeapon))

(event EventOnRemoval(Cascade)
  (params (MyTarget :type Actor%20Object%20Reference) (Parameters :type Gameplay%20Cue%20Parameters%20Structure%20%28by%20ref%29) (ParticleComponents :type Array%20of%20Cascade%20Particle%20System%20Component%20Object%20References) (AudioComponents :type Array%20of%20Audio%20Component%20Object%20References) (BurstCameraShakeInstance :type Camera%20Shake%20Base%20Object%20Reference) (BurstDecalInstance :type Decal%20Actor%20Object%20Reference))
  (bind _asfort_player_pawn (Utilities|Casting|CastToFortPlayerPawn MyTarget)
    (:then
      (Variables|Default|SetTargetPlayer _asfort_player_pawn))
    (:CastFailed
      _))
  (Utilities|FlowControl|Sequence
    (:then_0
      (CallFunction|ClearExpireTell))
    (:then_1
      (if (Pawn|IsLocallyControlled (Variables|Default|GetTargetPlayer))
        (Rendering|Material|SetScalarParameterValue 0 "/Game/Packages/Fortress_SharedMaterials/GlobalMaterialParameters/FortniteMaterialParameters.FortniteMaterialParameters" "SpookyMistMode" (Math|Conversions|ToFloat(Boolean) (Replay|GetIsPlayingReplay 0)) :!class /Script/Engine.KismetMaterialLibrary)
        (else
          (Utilities|IsValid (Variables|Default|GetTargetPlayer)
            (:Is%20Valid
              (CallFunction|UnbindEquippedWeapon)))))
      (Variables|Default|SetTargetPlayer 0))))

(fn GetSkeletalMeshes
  (bind _targetplayer (Variables|Default|GetTargetPlayer))
  (Utilities|Array|Clear (Variables|Default|GetPlayerSkeletalMeshes))
  (Utilities|IsValid _targetplayer
    (:Is%20Valid
      (for _array_element (Components|GetChildrenComponents (Variables|/Script/Engine.Character|GetMesh _targetplayer) false)
        (bind _asskeletal_mesh_component (Utilities|Casting|CastToSkeletalMeshComponent _array_element))
        (Utilities|Array|AddUnique (Variables|Default|GetPlayerSkeletalMeshes) _asskeletal_mesh_component)))
    (:Is%20Not%20Valid
      (Development|PrintString 0 "not valid")))
  (return))

(fn SetWeaponVisibility
  (params (FortWeapon :type Fort%20Weapon%20Object%20Reference) (Visible :type Boolean))
  (bind _returnvalue (Visual|GetWeaponMesh FortWeapon))
  (Utilities|IsValid FortWeapon
    (:Is%20Valid
      (bind _asb_ranged_dual_generic_athena (Utilities|Casting|CastToB_Ranged_Dual_Generic_Athena FortWeapon)
        (:then
          (Variables|/Game/Weapons/Blueprints/B_Ranged_Dual_Generic_Athena.B_Ranged_Dual_Generic_Athena_C|SetbShouldReEnableLeftWeapVis Visible _asb_ranged_dual_generic_athena))
        (:CastFailed
          (bind _asfort_weapon_pickaxe_dual_wield_athena (Utilities|Casting|CastToFortWeaponPickaxeDualWieldAthena FortWeapon)
            (:then
              (Utilities|IsValid (Variables|/Script/FortniteGame.FortWeaponPickaxeDualWieldAthena|GetWeaponMeshOffhand _asfort_weapon_pickaxe_dual_wield_athena)
                (:Is%20Valid
                  (Rendering|SetVisibility (Variables|/Script/FortniteGame.FortWeaponPickaxeDualWieldAthena|GetWeaponMeshOffhand _asfort_weapon_pickaxe_dual_wield_athena) Visible true))
                (:Is%20Not%20Valid
                  _))
              (Utilities|IsValid _returnvalue
                (:Is%20Valid
                  (Rendering|SetVisibility _returnvalue Visible true))))
            (:CastFailed
              _)))))))

(fn UnbindEquippedWeapon
  (bind _targetplayer (Variables|Default|GetTargetPlayer))
  (Utilities|IsValid _targetplayer
    (:Is%20Valid
      (FortPawn|UnbindEventfromOnWeaponEquippedDelegate _targetplayer (EventDispatchers|CreateEvent :!event OnWeaponEquippedDelegate_Event_0)))))

(fn ConstructionScript)
